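# Translucent, unlit, two-sided Niagara ribbon and mesh-particle material for arc airflow effects on the Ensemble Maroon glider.
# Emissive tints two sine-crossfaded, camera-plane noise layers with particle color.
# Opacity combines an airflow texture, scrolling noise and a rounded dissolve driven by dynamic parameters, or a distorted smoke trail, then depth-fades except on mobile.
# SplineThicken sets width through world position offset, with a parameter to disable it.
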Material /BRCosmetics/Gadgets/Assets/VinderTech_GliderChute/Glider_Ensemble_Maroon_Male/FX/Materials/M_EM_Glider_arc_airflow
graph {
 properties {
  bUsedWithNiagaraMeshParticles=true
  bUsedWithNiagaraRibbons=true
  blendMode=BLEND_Translucent
  pixelDepthOffsetMode=PDOM_Legacy
  shadingModel=MSM_Unlit
  twoSided=true
  usedShadingModels=Unlit
 }
 nodes [
  n1 = ScalarParameter() {defaultValue=0.4, group=35.TextureBlend, parameterName=Trd_RemapLow}
  n2 = Multiply(
   A=Time() {bOverride_Period=true, period=500}
   B=ScalarParameter() {defaultValue=0.5, group=35.TextureBlend, parameterName=Trd_ExtraLayerBlendSpeed}
  )
  n3 = LinearInterpolate(A=NamedRerouteUsage(), B=NamedRerouteUsage(), Alpha=ConstantBiasScale(None=Sine(None=@n2)))
  n4 = ParticleColor()
  MaterialFunctionCall(Emissive=Multiply(A=LinearInterpolate(A=@n1, Alpha=@n3), B=@n4.RGB)) {materialFunction=&MF_FX_EmissiveAdjustments}
  n5 = Time() {bOverride_Period=true, period=9999}
  n6 = VectorParameter() {group=32.TextureTex_Trd, parameterName=TextureTex_Trd_UVControl_ExtraLayer, defaultValue=<rgba 0.015 0.015 0 0>}
  n7 = Transform(None=Constant3Vector() {constant=<rgba 0 0 -1 0>}) {transformSourceType=TRANSFORMSOURCE_Camera}
  n8 = Transform(None=Constant3Vector() {constant=<rgba 1 0 0 0>}) {transformSourceType=TRANSFORMSOURCE_Camera}
  n9 = ObjectPositionWS()
  n10 = NamedRerouteDeclaration(None=MaterialFunctionCall(
    UVs=MaterialFunctionCall(Plane Normal Axis=@n7, Plane X Axis=@n8, Plane Center=@n9) {materialFunction=&VirtualPlaneCoordinates}.UVs
    Texture Scale=ScalarParameter() {defaultValue=1, group=00.Base, parameterName=Base_TextureTexUV}
   ) {materialFunction=&ScaleUVsByCenter}.UVs) {name=ScreenUV, nodeColor=<rgba 0 0.3203 1 1>}
  n11 = TextureObjectParameter() {
   group=32.TextureTex_Trd
   parameterName=TextureTex_Trd_Tex
   samplerType=SAMPLERTYPE_LinearGrayscale
   texture=/Game/Global/Textures/Default/Noise/T_WhiteNoise
  }
  NamedRerouteDeclaration(None=TextureSample!(
    UVs=MaterialFunctionCall!(UVInput=@n10, TimeInput=@n5, Control_RGB=@n6.RGB, Control_A=@n6.RGB) {materialFunction=/Game/Effects/Niagara/MasterMaterials/Functions/Common/FXCommon_QuickUVControl}.UV
    Tex=@n11
    Level
    Apply View MipBias
   ) {constMipValue=0, mipValueMode=TMVM_MipLevel, samplerType=SAMPLERTYPE_LinearColor}.R) {name=TextureTex_Trd_ExtraLayer, nodeColor=<rgba 0.3125 1 0 1>}
  n12 = VectorParameter() {group=32.TextureTex_Trd, parameterName=TextureTex_Trd_UVControl, defaultValue=<rgba 0.013 0.013 0 0>}
  NamedRerouteDeclaration(None=TextureSample(
    UVs=MaterialFunctionCall(UVInput=@n10, TimeInput=@n5, Control_RGB=@n12.RGB, Control_A=@n12.RGB) {materialFunction=&FXCommon_QuickUVControl}.UV
    Tex=@n11
   ) {constMipValue=0, mipValueMode=TMVM_MipLevel, samplerType=SAMPLERTYPE_LinearColor}.R) {name=TextureTex_Trd, nodeColor=<rgba 0.3125 1 0 1>}
  n13 = Multiply(A=Constant() {r=30}, B=DynamicParameter() {defaultValue=<rgba 0.5 1 1 10>, paramNames [Offset,Dissolve,Width,Noise_offset]}.Width)
  Reroute(None=LinearInterpolate(
    A=MaterialFunctionCall(WidthBase=@n13, WidthTip=@n13) {materialFunction=&SplineThicken}.WorldPositionOffset
    B=Constant()
    Alpha=ScalarParameter() {parameterName=No_SplineThicken}
   ))
  n14 = TextureCoordinate()
  n15 = VectorParameter() {parameterName=Tex_UVScale, defaultValue=<rgba 1 1 1 1>}
  n16 = AppendVector(A=@n15.R, B=@n15.R)
  n17 = DynamicParameter() {defaultValue=<rgba 0.5 1 1 10>, paramNames [Offset,Dissolve,Width,Noise_offset]}
  n18 = Constant()
  n19 = ScalarParameter() {defaultValue=1, parameterName=Tex_default_V_panner_speed}
  n20 = TextureSampleParameter2D(UVs=Panner(Coordinate=Multiply(A=@n14, B=@n16), Time=LinearInterpolate(Alpha=@n17.Offset) {constA=-1}, Speed=AppendVector(A=@n18, B=@n19)) {speedY=1}) {
   parameterName=Tex_Default
   samplerSource=SSM_Clamp_WorldGroupSettings
   texture=/Game/Effects/Fort_Effects/Emotes/Skirmish/Skirmish_Male/T_Skirmish_airflow
  }
  n21 = StaticComponentMaskParameter(None=@n20.RGB) {defaultR=true, parameterName=Component_mask}
  n22 = MaterialFunctionCall(XY=TextureCoordinate()) {materialFunction=&Swizzle}
  n23 = VectorParameter() {parameterName=Tex_noise_UVScale, defaultValue=<rgba 0.5 1 1 1>}
  n24 = AppendVector(A=@n23.R, B=@n23.R)
  n25 = DynamicParameter() {defaultValue=<rgba 0.5 1 1 10>, paramNames [Offset,Dissolve,Width,Noise_offset]}
  n26 = Time()
  n27 = ScalarParameter() {defaultValue=1, parameterName=Loop_time}
  n28 = TextureSampleParameter2D(UVs=Panner(Coordinate=Multiply(A=@n22.YX, B=@n24), Time=LinearInterpolate(A=@n25.Noise_offset, B=@n26, Alpha=@n27)) {speedX=1.5}) {parameterName=Tex_noise, samplerType=SAMPLERTYPE_Grayscale, texture=&T_Smoke_Trail_Soft}
  n29 = Multiply(A=Multiply(A=@n28.R, B=@n28.R), B=ScalarParameter() {defaultValue=2, parameterName=Tex_noise_brightness}) {constB=2}
  n30 = Power(
   Base=Multiply(A=Multiply(A=@n21, B=@n29)) {constB=2}
   Exp=ScalarParameter() {defaultValue=1, group=Only_Use_Legs_fresnel, parameterName=power}
  ) {constExponent=1}
  n31 = ScalarParameter() {parameterName=Tex_noise_lerp_dark}
  n32 = VectorParameter() {parameterName=Dissolve, defaultValue=<rgba 0.3 0 1 1>}
  n33 = TextureSampleParameter2D(UVs=Add(A=TextureCoordinate() {vTiling=2}, B=AppendVector(A=@n32.R, B=@n32.R))) {parameterName=Tex_Dissolve, texture=/Game/Global/Textures/Default/Masks_and_Gradients/T_BlasterBeamDistort}
  n34 = Power(Base=Multiply(A=Add(A=@n30) {constB=0}, B=LinearInterpolate(A=@n31, Alpha=@n33.R))) {constExponent=1.5}
  n35 = ScalarParameter() {defaultValue=1, parameterName=Multiply_noise}
  n36 = Sine(None=ComponentMask(None=TextureCoordinate()) {g=true}) {period=2}
  n37 = Multiply(A=Multiply(A=LinearInterpolate(A=@n21, B=@n34, Alpha=@n35), B=Power(Base=@n36)))
  n38 = VectorParameter() {parameterName=Dissolve_UVScale, defaultValue=<rgba 1 1 1 1>}
  n39 = VectorParameter() {parameterName=Dissolve, defaultValue=<rgba 0.3 0 1 1>}
  n40 = Add(A=Multiply(A=TextureCoordinate(), B=AppendVector(A=@n38.R, B=@n38.R)), B=AppendVector(A=@n39.R, B=@n39.R))
  n41 = Power(
   Base=ComponentMask(None=TextureCoordinate()) {r=true}
   Exp=LinearInterpolate(Alpha=DynamicParameter() {defaultValue=<rgba 0 1 1 0>, paramNames [Offset,Dissolve,Width,Param4]}.Dissolve) {constB=5}
  )
  n42 = Round(None=LinearInterpolate(
    A=TextureSampleParameter2D(UVs=@n40) {parameterName=Tex_Dissolve, texture=&T_BlasterBeamDistort}.R
    B=Saturate(None=@n41)
   ))
  n43 = TextureSampleParameter2D(UVs=Panner(Coordinate=TextureCoordinate() {vTiling=0.5}) {speedY=1}) {
   group=Noise
   parameterName=Noise Texture
   samplerType=SAMPLERTYPE_Normal
   texture=&T_Weap_Distort_N_02
  }
  n44 = Multiply(A=ComponentMask(None=@n43.RGB) {g=true, r=true}) {constB=0.2}
  n45 = TextureCoordinate()
  n46 = DynamicParameter() {desc=niagara dynamic 1, parameterIndex=1, defaultValue=<rgba 1 1 0 0>, paramNames [Power_and_UVscale, offset, Noise offset, Param4]}
  n47 = AppendVector(A=Constant() {r=1}, B=LinearInterpolate(Alpha=@n46.Power_and_UVscale) {constA=1, constB=0.7})
  n48 = Panner(Coordinate=MaterialFunctionCall(XY=Multiply(A=@n45, B=@n47)) {materialFunction=&Swizzle}.YX) {speedX=0.8}
  n49 = Power(
   Base=TextureSample!(UVs=Add(A=@n44, B=@n48), Tex, Apply View MipBias) {samplerType=SAMPLERTYPE_Grayscale, texture=&T_Smoke_Trail_Soft}.R
   Exp=LinearInterpolate(Alpha=@n46.Power_and_UVscale) {constA=1, constB=1.5}
  ) {constExponent=1.5}
  n50 = TextureCoordinate()
  n51 = Power(Base=ComponentMask(None=@n50) {g=true}) {constExponent=1}
  n52 = Multiply(A=OneMinus(None=ComponentMask(None=@n50) {g=true}), B=Saturate(None=Multiply(A=@n51) {constB=10}))
  n53 = Multiply(
   A=@n4.A
   B=LinearInterpolate(A=Subtract(A=@n37, B=@n42), B=Multiply(A=@n49, B=@n52), Alpha=ScalarParameter() {parameterName=Use disturbed trail})
  )
  n54 = DepthFade(Opacity=@n53, FadeDistance=ScalarParameter() {defaultValue=50, parameterName=DepthFade}) {fadeDistanceDefault=50}
  ShadingPathSwitch(Default=@n54, Deferred=@n54, Forward=@n54, Mobile=@n53)
 ]
}
