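# Masked, lit Niagara sprite material for the VinderTech GliderChute effects, named for downward-flowing sand.
# A droplet texture with flipped, noise-distorted UVs multiplies a hard threshold of inverted noise plus a particle parameter to form a dissolving shape.
# Color is tinted by a sine-driven noise blend, apparently two white-noise layers on camera-aligned virtual-plane UVs, feeding emissive adjustments.
# Dynamic parameters drive noise time, threshold, hue shift, desaturation and brightness per particle.
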Material /BRCosmetics/Gadgets/Assets/VinderTech_GliderChute/Glider_Ensemble_Maroon_Male/FX/Materials/M_FlowDownSand
graph {
 properties {bUsedWithNiagaraSprites=true, blendMode=BLEND_Masked, pixelDepthOffsetMode=PDOM_Legacy, usedShadingModels=Default Lit}
 nodes [
  n1 = DynamicParameter() {paramNames [Param1,Param2,Desat,Scale]}
  Add(A=Time() {bOverride_Period=true, period=9999}, B=@n1.Param1)
  Constant() {r=1}
  n2 = ScalarParameter() {defaultValue=0.6, group=35.TextureBlend, parameterName=Trd_RemapLow}
  n3 = Multiply(A=Time(), B=ScalarParameter() {defaultValue=0.5, group=35.TextureBlend, parameterName=Trd_ExtraLayerBlendSpeed})
  n4 = LinearInterpolate(A=NamedRerouteUsage(), B=NamedRerouteUsage(), Alpha=ConstantBiasScale(None=Sine(None=@n3)))
  n5 = MaterialFunctionCall(
   Hue Shift Percentage=DynamicParameter() {parameterIndex=1, paramNames [hue,Param2,Param3,Param4]}.hue
   Texture=Multiply(A=VectorParameter() {parameterName=Color, defaultValue=<rgba 0.23 0.1224 0.0644 1>}.RGB, B=LinearInterpolate(A=@n2, Alpha=@n4))
  ) {materialFunction=&HueShift}
  MaterialFunctionCall(Emissive=Multiply(A=Desaturation(None=@n5.Result, Fraction=@n1.Desat), B=@n1.Scale)) {materialFunction=&MF_FX_EmissiveAdjustments}
  n6 = VectorParameter() {parameterName=UVControl, defaultValue=<rgba 3 1 -0.05 -0.5>}
  n7 = MaterialFunctionCall(
   Texture=TextureObjectParameter() {parameterName=NoiseTex, texture=/Game/Effects/Fort_Effects/Emotes/UltraLight/Texture/T_SoftNoise_BC7}
   UVControl_RGB=@n6.RGB
   UVControl_A=@n6.RGB
   "[Override]TimeInput"=@n1.Param1
  ) {materialFunction=&FXCommon_TextureSample}
  n8 = OneMinus(None=@n7.R)
  n9 = Constant()
  n10 = MaterialFunctionCall(Float2=TextureCoordinate()) {materialFunction=&BreakOutFloat2Components}
  n11 = MaterialFunctionCall(X=@n10.R, Y=OneMinus(None=@n10.G)) {materialFunction=&MakeFloat2}
  n12 = ScalarParameter() {defaultValue=0.6}
  n13 = VectorParameter() {parameterName="NoiseStrength]", defaultValue=<rgba -0.5 -0.3 0 0.5>}
  n14 = Add(A=AppendVector(A=@n13.R, B=@n13.R), B=AppendVector(A=@n7.R, B=@n7.R))
  n15 = AppendVector(A=@n13.B, B=@n13.B)
  n16 = TextureSample(UVs=Add(
    A=MaterialFunctionCall(UVs=@n11.Result, Texture Scale=@n12) {materialFunction=&ScaleUVsByCenter}.UVs
    B=Multiply(A=@n14, B=@n15)
   )) {samplerSource=SSM_Clamp_WorldGroupSettings, texture=/BRCosmetics/Athena/Effects/Particles/Skydiving/NS_Hightower_Date/T_Hightower_Droplet}
  Multiply(A=If(A=Add(A=@n8, B=@n1.Param2), B=@n9, "A > B"=Constant() {r=1}, "A == B"=@n9, "A < B"=@n9), B=@n16.R)
  n17 = NamedRerouteUsage()
  n18 = VectorParameter() {group=32.TextureTex_Trd, parameterName=TextureTex_Trd_UVControl_ExtraLayer, defaultValue=<rgba 0.015 0.015 0 0>}
  n19 = Transform(None=Constant3Vector() {constant=<rgba 0 0 -1 0>}) {transformSourceType=TRANSFORMSOURCE_Camera}
  n20 = Transform(None=Constant3Vector() {constant=<rgba 1 0 0 0>}) {transformSourceType=TRANSFORMSOURCE_Camera}
  n21 = ObjectPositionWS()
  n22 = NamedRerouteDeclaration(None=MaterialFunctionCall(
    UVs=MaterialFunctionCall(Plane Normal Axis=@n19, Plane X Axis=@n20, Plane Center=@n21) {materialFunction=&VirtualPlaneCoordinates}.UVs
    Texture Scale=ScalarParameter() {defaultValue=1, group=00.Base, parameterName=Base_TextureTexUV}
   ) {materialFunction=&ScaleUVsByCenter}.UVs) {name=ScreenUV, nodeColor=<rgba 0 0.3203 1 1>}
  n23 = TextureObjectParameter() {
   group=32.TextureTex_Trd
   parameterName=TextureTex_Trd_Tex
   samplerType=SAMPLERTYPE_LinearGrayscale
   texture=/Game/Global/Textures/Default/Noise/T_WhiteNoise
  }
  NamedRerouteDeclaration(None=TextureSample!(
    UVs=MaterialFunctionCall!(UVInput=@n22, TimeInput=@n17, Control_RGB=@n18.RGB, Control_A=@n18.RGB) {materialFunction=/Game/Effects/Niagara/MasterMaterials/Functions/Common/FXCommon_QuickUVControl}.UV
    Tex=@n23
    Level
    Apply View MipBias
   ) {constMipValue=0, mipValueMode=TMVM_MipLevel, samplerType=SAMPLERTYPE_LinearColor}.R) {name=TextureTex_Trd_ExtraLayer, nodeColor=<rgba 0.3125 1 0 1>}
  n24 = VectorParameter() {group=32.TextureTex_Trd, parameterName=TextureTex_Trd_UVControl, defaultValue=<rgba 0.013 0.013 0 0>}
  NamedRerouteDeclaration(None=TextureSample(
    UVs=MaterialFunctionCall(UVInput=@n22, TimeInput=@n17, Control_RGB=@n24.RGB, Control_A=@n24.RGB) {materialFunction=&FXCommon_QuickUVControl}.UV
    Tex=@n23
   ) {constMipValue=0, mipValueMode=TMVM_MipLevel, samplerType=SAMPLERTYPE_LinearColor}.R) {name=TextureTex_Trd, nodeColor=<rgba 0.3125 1 0 1>}
  ScalarParameter() {parameterName=DissolveTest}
 ]
}
