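# M_HideMesh is a material in the Maroon Male glider FX folder, named for hiding a mesh.
# It is a masked, unlit, wireframe material enabled for skeletal meshes and mesh deformers.
# The graph holds two scalar parameters, one named Param_1, with no connections shown to each other or to any material output, so no opacity mask logic is visible.
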Material /BRCosmetics/Gadgets/Assets/VinderTech_GliderChute/Glider_Ensemble_Maroon_Male/FX/Materials/M_HideMesh
graph {
 properties {
  bUsedWithMeshDeformer=true
  bUsedWithSkeletalMesh=true
  blendMode=BLEND_Masked
  pixelDepthOffsetMode=PDOM_Legacy
  shadingModel=MSM_Unlit
  usedShadingModels=Unlit
  wireframe=true
 }
 nodes [
  ScalarParameter()
  ScalarParameter() {parameterName=Param_1}
 ]
}
