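# M_SandTrail is an unlit translucent Niagara mesh-particle material for a sand trail effect on a glider.
# Scrolling noise textures with UV distortion build three tinted layers, adjusted by dynamic-parameter hue, desaturation and scale.
# Opacity combines noise, UV edge masks, backface culling, particle alpha and custom-depth fading on the deferred path.
# Vertex offset uses FollowingDeform and SplineThicken functions, driven by particle follow position, rest axis, stretch and width.
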Material /BRCosmetics/Gadgets/Assets/VinderTech_GliderChute/Glider_Ensemble_Maroon_Male/FX/Materials/M_SandTrail
graph {
 properties {
  bEnableResponsiveAA=true
  bUsedWithNiagaraMeshParticles=true
  blendMode=BLEND_Translucent
  pixelDepthOffsetMode=PDOM_Legacy
  shadingModel=MSM_Unlit
  twoSided=true
  usedShadingModels=Unlit
 }
 nodes [
  n1 = DynamicParameter() {parameterIndex=1, paramNames [Desat, ScaleC, Hue, ""]}
  n2 = MaterialFunctionCall(Float2=TextureCoordinate()) {materialFunction=&BreakOutFloat2Components}
  n3 = VectorParameter() {parameterName="EdgeControl (V3)1111", defaultValue=<rgba 0 1 1 0>}
  n4 = VectorParameter() {parameterName=WPO_Sec_UVControl, defaultValue=<rgba 1 1 0 0>}
  n5 = Add(A=Time(), B=ScalarParameter() {parameterName=r})
  n6 = MaterialFunctionCall(
   Texture=TextureObjectParameter() {parameterName=WPO_Sec_HeightTex, texture=/Game/Global/Textures/Default/Noise/T_Tiling_Noise_01}
   UVControl_RGB=@n4.RGB
   UVControl_A=@n4.RGB
   "[Override]TimeInput"=@n5
  ) {materialFunction=&FXCommon_TextureSample}
  n7 = Subtract(A=@n6.R, B=ScalarParameter() {defaultValue=0.25, parameterName=Noffset_B}) {constB=0.25}
  n8 = VectorParameter() {parameterName=WPO_Sec_UVN, defaultValue=<rgba 1 1 0 0>}
  n9 = TextureCoordinate()
  n10 = Add(A=Multiply(A=AppendVector(A=@n7, B=@n7), B=AppendVector(A=@n8.R, B=@n8.R)), B=@n9)
  n11 = VectorParameter() {parameterName=Texture3333UVControl, defaultValue=<rgba 1 1 0 0>}
  n12 = TextureObjectParameter() {parameterName=Texture333_Tex, samplerType=SAMPLERTYPE_LinearGrayscale, texture=/Game/Global/Textures/Default/Noise/T_WhiteNoise}
  n13 = TextureSample!(
   UVs=MaterialFunctionCall!(UVInput=@n10, TimeInput=@n5, Control_RGB=@n11.RGB, Control_A=@n11.RGB) {materialFunction=/Game/Effects/Niagara/MasterMaterials/Functions/Common/FXCommon_QuickUVControl}.UV
   Tex=@n12
   Level
   Apply View MipBias
  ) {constMipValue=0, mipValueMode=TMVM_MipLevel, samplerType=SAMPLERTYPE_LinearColor}
  n14 = ScalarParameter() {parameterName=W3}
  n15 = Multiply(
   A=@n6.R
   B=MaterialFunctionCall(In=Power(Base=@n13.R, Exp=@n14), Contrast=ScalarParameter() {parameterName=c3}) {materialFunction=&CheapContrast}.Result
  )
  n16 = VectorParameter() {parameterName=C3, defaultValue=<rgba 0.1 0.041 0.0115 1>}
  n17 = VectorParameter() {parameterName=C1, defaultValue=<rgba 0.1 0.041 0.0115 1>}
  n18 = TextureCoordinate()
  n19 = Add(
   A=ScalarParameter() {parameterName=TP}
   B=Multiply(
    A=MaterialFunctionCall(Float2=@n18) {materialFunction=&BreakOutFloat2Components}.G
    B=ScalarParameter() {parameterName=TTTM}
   )
  )
  n20 = Add(A=Multiply(A=Saturate(None=@n19), B=ScalarParameter() {parameterName=TTT}))
  n21 = VectorParameter() {parameterName=NNNc_UVControl, defaultValue=<rgba 1 1 0 0>}
  n22 = Subtract(
   A=MaterialFunctionCall(
    Texture=TextureObjectParameter() {parameterName=WPO_NNN_HeightTex, texture=&T_Tiling_Noise_01}
    UVControl_RGB=@n21.RGB
    UVControl_A=@n21.RGB
    "[Override]TimeInput"=@n5
   ) {materialFunction=&FXCommon_TextureSample}.R
   B=ScalarParameter() {defaultValue=0.3, parameterName=Noffset_A}
  ) {constB=0.3}
  n23 = VectorParameter() {parameterName=NNNN_UVN, defaultValue=<rgba 1 1 0 0>}
  n24 = Multiply(A=AppendVector(A=@n22, B=@n22), B=AppendVector(A=@n23.R, B=@n23.R))
  n25 = VectorParameter() {parameterName=Texture_Trd_UVControl, defaultValue=<rgba 1 1 0 0>}
  n26 = MaterialFunctionCall(UVInput=Add(A=@n24, B=@n9), TimeInput=@n5, Control_RGB=@n25.RGB, Control_A=@n25.RGB) {materialFunction=&FXCommon_QuickUVControl}
  n27 = TextureObjectParameter() {parameterName=Texture_Trd_Tex, samplerType=SAMPLERTYPE_LinearGrayscale, texture=&T_WhiteNoise}
  n28 = MaterialFunctionCall(
   In=Power(
    Base=TextureSample(UVs=@n26.UV, Tex=@n27) {constMipValue=0, mipValueMode=TMVM_MipLevel, samplerType=SAMPLERTYPE_LinearColor}.R
    Exp=ScalarParameter() {parameterName=W1}
   )
   Contrast=ScalarParameter() {parameterName=c1}
  ) {materialFunction=&CheapContrast}
  n29 = Add(
   A=Multiply(
    A=MaterialFunctionCall!(UorV=@n2.R, EdgeControl=@n3.RGB) {materialFunction=/Game/Effects/Niagara/MasterMaterials/Functions/Common/FXCommon_EdgeMask_Simple}.Mask
    B=Multiply(A=@n15, B=@n16.RGB)
   )
   B=Multiply(A=Multiply(A=@n17.RGB, B=@n20), B=@n28.Result)
  )
  n30 = VectorParameter() {parameterName=TextureSS_UVControl, defaultValue=<rgba 1 1 0 0>}
  n31 = MaterialFunctionCall(UVInput=Add(A=@n24, B=@n9), TimeInput=@n5, Control_RGB=@n30.RGB, Control_A=@n30.RGB) {materialFunction=&FXCommon_QuickUVControl}
  n32 = TextureObjectParameter() {parameterName=TextureSS_Tex, samplerType=SAMPLERTYPE_LinearGrayscale, texture=&T_WhiteNoise}
  n33 = MaterialFunctionCall(
   In=Power(
    Base=TextureSample(UVs=@n31.UV, Tex=@n32) {constMipValue=0, mipValueMode=TMVM_MipLevel, samplerType=SAMPLERTYPE_LinearColor}.RGB
    Exp=ScalarParameter() {parameterName=W2}
   )
   Contrast=ScalarParameter() {parameterName=c2}
  ) {materialFunction=&CheapContrast}
  n34 = Multiply(A=Multiply(A=@n20, B=VectorParameter() {parameterName=C2, defaultValue=<rgba 0.1 0.041 0.0115 1>}.RGB), B=@n33.Result)
  n35 = ScalarParameter() {parameterName=Debug}
  n36 = OneMinus(None=MaterialFunctionCall(Float2=TextureCoordinate()) {materialFunction=&BreakOutFloat2Components}.G)
  n37 = MaterialFunctionCall(Hue Shift Percentage=@n1.Hue, Texture=LinearInterpolate(A=Add(A=@n29, B=@n34), B=@n36, Alpha=@n35)) {materialFunction=&HueShift}
  MaterialFunctionCall(Emissive=Multiply(A=Desaturation(None=@n37.Result, Fraction=@n1.Desat), B=@n1.ScaleC)) {materialFunction=&MF_FX_EmissiveAdjustments}
  n38 = MaterialFunctionCall(Float2=@n10) {materialFunction=&BreakOutFloat2Components}
  n39 = MaterialFunctionCall(UorV=@n38.R, EdgeControl=VectorParameter() {parameterName="EdgeControl (V3)", defaultValue=<rgba 0 1 1 0>}.RGB) {materialFunction=&FXCommon_EdgeMask_Simple}
  n40 = MaterialFunctionCall!(
   UorV=@n38.G
   Invert_EdgeControl=VectorParameter() {parameterName="EdgeControl 2(V3)", defaultValue=<rgba 0 1 1 0>}.RGB
   NoInvert_EdgeControl=VectorParameter() {parameterName="EdgeControl 3(V3)", defaultValue=<rgba 0 1 1 0>}.RGB
  ) {materialFunction=/Game/Effects/Niagara/MasterMaterials/Functions/Common/FXCommon_EdgeMask_Ex}
  Min(A=@n39.Mask, B=@n40.Mask)
  n41 = NamedRerouteUsage()
  n42 = Multiply(A=@n28.Result, B=ScalarParameter() {parameterName=ssad1})
  n43 = Multiply(A=@n33.Result, B=ScalarParameter() {parameterName=ssad2})
  n44 = Multiply(
   A=Multiply(A=Multiply(A=@n39.Mask, B=@n40.Mask), B=Multiply(A=@n42, B=@n43))
   B=MaterialFunctionCall!(ObjectOrParticlePosition=ParticlePositionWS()) {materialFunction=/Game/Effects/Fort_Effects/Materials/Toon/CustomDepthOutLine/MF_BackfaceCulling}.Result
  )
  Multiply(A=ParticleColor().A, B=LinearInterpolate(A=Multiply(A=@n41, B=@n44), Alpha=@n35))
  n45 = MaterialFunctionCall(Float3=SceneTexture() {sceneTextureId=PPI_CustomDepth}.Color) {materialFunction=&BreakOutFloat3Components}
  n46 = PixelDepth()
  n47 = ComponentMask(None=@n46.R) {r=true}
  n48 = Saturate(None=Divide(A=Subtract(A=@n45.R, B=@n47), B=ScalarParameter() {defaultValue=50, group=05.CustomDepthCullAndFade, parameterName=CustomDepthFadeDistance}) {constB=50})
  n49 = Constant() {r=1}
  n50 = If(A=@n46.R, B=@n45.R, "A > B"=Constant(), "A == B"=@n49, "A < B"=@n49) {equalsThreshold=0.1}
  n51 = Constant() {r=1}
  NamedRerouteDeclaration(None=ShadingPathSwitch(Default=Multiply(A=@n48, B=@n50), Forward=@n51, Mobile=@n51)) {name=R_CustomDepthCullAndFade, nodeColor=<rgba 0 0.6953 1 1>}
  ObjectPositionWS()
  n52 = DynamicParameter() {parameterIndex=2, paramNames [FollowPos_X, FollowPos_Y, FollowPos_Z, StretchDistance]}
  n53 = DynamicParameter() {parameterIndex=3, paramNames [RestAxis_X,RestAxis_Y,RestAxis_Z,Param4]}
  n54 = MaterialFunctionCall!(
   Mask=Saturate(None=@n36)
   FollowPos=AppendVector(A=AppendVector(A=@n52.FollowPos_X, B=@n52.FollowPos_X), B=@n52.FollowPos_Z)
   RestAxis=AppendVector(A=AppendVector(A=@n53.RestAxis_X, B=@n53.RestAxis_X), B=@n53.RestAxis_Z)
   MeshRootPos
   NormalWS
   StretchDistance=@n52.StretchDistance
   Rigidity
  ) {materialFunction=/Game/Effects/Niagara/Modules/FollowingDeform/Material/MF_NiagaraFollowingDeform}
  n55 = DynamicParameter() {paramNames [ScaleX,Age,Width,Param4]}
  n56 = MaterialFunctionCall!(
   TangentU_WS=@n54.ForSplineThicken_TangentU_WS
   TangentV_WS=@n54.ForSplineThicken_TangentU_WS
   WidthBase=Multiply(A=@n55.Width, B=ScalarParameter())
   WidthTip=Multiply(A=@n55.Width, B=ScalarParameter() {parameterName=Param_1})
   WorldPosition=Add(A=@n54.WPO, B=WorldPosition() {worldPositionShaderOffset=WPT_ExcludeAllShaderOffsets}.XYZ)
   UVs for Projection
   UVs for Thickness
   UVs for texturing
   Expand U or V UV Channel=StaticBool() {value=true}
   DeriveNormalZ
   AdditionalNormal
   AngleCorrectedNormals
   FlattenPixelNormal
   UVs For NormalShading
  ) {materialFunction=/Game/Effects/Niagara/Modules/FollowingDeform/Material/SplineThickenWithTangentInput}
  n57 = DynamicParameter() {paramNames [ScaleX,Age,Width,Param4]}
  Reroute(None=Add(A=@n54.WPO, B=Multiply(A=@n56.WorldPositionOffset, B=@n57.ScaleX))) {desc=WPO}
  TextureCoordinate()
 ]
}
