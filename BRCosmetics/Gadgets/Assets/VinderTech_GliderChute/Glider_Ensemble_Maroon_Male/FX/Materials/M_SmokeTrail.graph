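# Unlit, two-sided translucent material for Niagara mesh particles, forming the VinderTech glider's smoke trail.
# Vertex offset combines a Niagara following deform with tangent-fed spline thickening, driven by dynamic parameters.
# Two panning noises and a stretch term distort a grayscale texture for two-color emissive, with screen-aligned grain layers and dynamic hue, desaturation and scale.
# Opacity combines that texture, inverted Fresnel, edge mask, backface culling and particle alpha, with custom-depth fade declared.
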Material /BRCosmetics/Gadgets/Assets/VinderTech_GliderChute/Glider_Ensemble_Maroon_Male/FX/Materials/M_SmokeTrail
graph {
 properties {
  bUsedWithNiagaraMeshParticles=true
  blendMode=BLEND_Translucent
  pixelDepthOffsetMode=PDOM_Legacy
  shadingModel=MSM_Unlit
  twoSided=true
  usedShadingModels=Unlit
 }
 nodes [
  n1 = NamedRerouteUsage()
  n2 = NamedRerouteUsage()
  n3 = Time() {bOverride_Period=true, period=500}
  n4 = ScalarParameter() {defaultValue=0.5, group=35.TextureBlend, parameterName=Trd_ExtraLayerBlendSpeed}
  n5 = ConstantBiasScale(None=Sine(None=Multiply(A=@n3, B=@n4)))
  n6 = VectorParameter() {group=MainTexture, parameterName=MainTexture_G2C_ColorA, defaultValue=<rgba 0.135 0.0811 0.0499 1>}
  n7 = VectorParameter() {group=MainTexture, parameterName=MainTexture_G2C_ColorB, defaultValue=<rgba 0.27 0.1623 0.0999 1>}
  n8 = TextureObjectParameter() {
   group=MainTexture
   parameterName=MainTexture_Texture
   samplerType=SAMPLERTYPE_Grayscale
   sortPriority=0
   texture=/Game/Gadgets/Assets/VinderTech_GliderChute/Glider_Thrive_Female/FX/Texture/T_RibbonBaseShape5
  }
  n9 = VectorParameter() {group=MainTexture, parameterName=MainTexture_UVControl, channelNames {a=VPan, b=UPan, g=VScale, r=UScale}, defaultValue=<rgba 1 1 0 0>}
  n10 = TextureObjectParameter() {
   group=NoiseB
   parameterName=NoiseB_Texture
   samplerType=SAMPLERTYPE_LinearGrayscale
   sortPriority=0
   texture=&T_SimplePerlin_Greyscale
  }
  n11 = VectorParameter() {group=NoiseB, parameterName=NoiseB_UVControl, channelNames {a=VPan, b=UPan, g=VScale, r=UScale}, defaultValue=<rgba 1 0.25 0.5 0.5>}
  n12 = NamedRerouteUsage()
  n13 = MaterialFunctionCall(Texture=@n10, UVControl_RGB=@n11, UVControl_A=@n11, "[Override]TimeInput"=NamedRerouteUsage(), "[Override]UVInput"=@n12) {materialFunction=&FXCommon_TextureSample}
  n14 = VectorParameter() {
   group=NoiseB
   parameterName=NoiseB_NoiseControl
   sortPriority=60
   channelNames {a=VStrength, b=UStrength, g=VOffset, r=UOffset}
   defaultValue=<rgba -0.5 -0.5 0.15 0.15>
  }
  n15 = MaterialFunctionCall!(NoiseValue=AppendVector(A=@n13.R, B=@n13.R), Control_RGB=@n14, Control_A=@n14) {materialFunction=/Game/Effects/Niagara/MasterMaterials/Functions/Common/FXCommon_NoiseControl}
  n16 = TextureObjectParameter() {
   group=NoiseA
   parameterName=NoiseA_Texture
   samplerType=SAMPLERTYPE_LinearGrayscale
   sortPriority=0
   texture=&T_SimplePerlin_Greyscale
  }
  n17 = VectorParameter() {group=NoiseA, parameterName=NoiseA_UVControl, channelNames {a=VPan, b=UPan, g=VScale, r=UScale}, defaultValue=<rgba 1.5 2 0.6 0>}
  n18 = MaterialFunctionCall(Texture=@n16, UVControl_RGB=@n17, UVControl_A=@n17, "[Override]TimeInput"=NamedRerouteUsage(), "[Override]UVInput"=@n12) {materialFunction=&FXCommon_TextureSample}
  n19 = VectorParameter() {
   group=NoiseA
   parameterName=NoiseA_NoiseControl
   sortPriority=60
   channelNames {a=VStrength, b=UStrength, g=VOffset, r=UOffset}
   defaultValue=<rgba -0.5 0 0.12 0>
  }
  n20 = MaterialFunctionCall(NoiseValue=AppendVector(A=@n18.R, B=@n18.R), Control_RGB=@n19, Control_A=@n19) {materialFunction=&FXCommon_NoiseControl}
  n21 = OneMinus(None=MaterialFunctionCall(Float2=NamedRerouteUsage()) {materialFunction=&BreakOutFloat2Components}.R)
  n22 = ScalarParameter() {defaultValue=0.05, group=Stretch, parameterName=Stretch_Offset}
  n23 = AppendVector(A=Multiply(A=Add(A=@n21, B=@n22), B=ScalarParameter() {defaultValue=0.7, group=Stretch, parameterName=Stretch_Strength}), B=Constant())
  n24 = NamedRerouteUsage()
  n25 = MaterialFunctionCall(
   Texture=@n8
   UVControl_RGB=@n9
   UVControl_A=@n9
   "[Override]TimeInput"=NamedRerouteUsage()
   "[Override]UVInput"=Add(A=Add(A=@n15.NoiseValue, B=@n20.NoiseValue), B=Add(A=@n23, B=@n24))
  ) {materialFunction=&FXCommon_TextureSample}
  n26 = OneMinus(None=Power(Base=OneMinus(None=@n25.R), Exp=ScalarParameter() {defaultValue=1, group=MainTexture, parameterName=MainTexture_G2C_InvtPower}))
  n27 = VectorParameter() {group=MainTexture, parameterName=MainTexture_G2C_Refine, channelNames {a=Blend, b=B_Mul, g=A_Mul, r=A_Power}}
  n28 = MaterialFunctionCall(Texture=@n26, RGB=@n27, A=@n27) {materialFunction=&MF_FX_Refine}
  n29 = Multiply(
   A=LinearInterpolate(A=ScalarParameter() {defaultValue=0.45, group=35.TextureBlend, parameterName=Trd_RemapLow}, Alpha=LinearInterpolate(A=@n1, B=@n2, Alpha=@n5))
   B=Multiply(A=LinearInterpolate(A=@n6.RGB, B=@n7.RGB, Alpha=@n26), B=Add(A=@n28.Result))
  )
  Desaturation(None=@n29)
  n30 = DynamicParameter() {parameterIndex=1, paramNames [Desat, ScaleC, Hue, ""]}
  n31 = MaterialFunctionCall(Hue Shift Percentage=@n30.Hue, Texture=@n29) {materialFunction=&HueShift}
  MaterialFunctionCall(Emissive=Multiply(A=Desaturation(None=@n31.Result, Fraction=@n30.Desat), B=@n30.ScaleC)) {materialFunction=&MF_FX_EmissiveAdjustments}
  MaterialFunctionCall(FlattenPixelNormal=StaticBool() {value=true}) {materialFunction=&SplineThicken}
  n32 = ScalarParameter() {defaultValue=4, parameterName=Fresnel_Exp}
  n33 = Constant()
  n34 = MaterialFunctionCall(Float2=NamedRerouteUsage()) {materialFunction=&BreakOutFloat2Components}
  n35 = DynamicParameter() {parameterIndex=2, paramNames [FollowPos_X, FollowPos_Y, FollowPos_Z, StretchDistance]}
  n36 = DynamicParameter() {parameterIndex=3, paramNames [RestAxis_X,RestAxis_Y,RestAxis_Z,Param4]}
  n37 = MaterialFunctionCall!(
   Mask=Saturate(None=OneMinus(None=@n34.R))
   FollowPos=AppendVector(A=AppendVector(A=@n35.FollowPos_X, B=@n35.FollowPos_X), B=@n35.FollowPos_Z)
   RestAxis=AppendVector(A=AppendVector(A=@n36.RestAxis_X, B=@n36.RestAxis_X), B=@n36.RestAxis_Z)
   MeshRootPos
   NormalWS
   StretchDistance=@n35.StretchDistance
   Rigidity
  ) {materialFunction=/Game/Effects/Niagara/Modules/FollowingDeform/Material/MF_NiagaraFollowingDeform}
  n38 = MaterialFunctionCall(Float2=NamedRerouteUsage()) {materialFunction=&BreakOutFloat2Components}
  n39 = ScalarParameter() {defaultValue=2, parameterName=WidthPwo}
  n40 = Multiply(
   A=LinearInterpolate(A=ScalarParameter() {defaultValue=340, parameterName=Width}, B=ScalarParameter() {parameterName=WidthB}, Alpha=Power(Base=@n38.R, Exp=@n39))
   B=DynamicParameter() {paramNames [ScaleX,Age,Width,Param4]}.Width
  )
  n41 = MaterialFunctionCall!(
   TangentU_WS=@n37.ForSplineThicken_TangentU_WS
   TangentV_WS=@n37.ForSplineThicken_TangentU_WS
   WidthBase=@n40
   WidthTip=@n40
   WorldPosition=Add(A=@n37.WPO, B=WorldPosition() {worldPositionShaderOffset=WPT_ExcludeAllShaderOffsets}.XYZ)
   UVs for Projection
   UVs for Thickness
   UVs for texturing
   Expand U or V UV Channel
   DeriveNormalZ
   AdditionalNormal
   AngleCorrectedNormals
   FlattenPixelNormal=StaticBool() {value=true}
   UVs For NormalShading
  ) {materialFunction=/Game/Effects/Niagara/Modules/FollowingDeform/Material/SplineThickenWithTangentInput}
  n42 = Power(
   Base=OneMinus(None=Fresnel(ExponentIn=@n32, BaseReflectFractionIn=@n33, Normal=@n41.Normal))
   Exp=ScalarParameter() {defaultValue=3, parameterName=Fresnel_Exp2}
  )
  n43 = Saturate(None=Power(Base=Saturate(None=@n25.R), Exp=ScalarParameter() {defaultValue=1, group=MainTexture, parameterName=MainTexture_ToOpacity_Power}))
  n44 = ScalarParameter() {defaultValue=0.7, group=MainTexture, parameterName=MainTexture_ToOpacity_MulAfterPower}
  n45 = MaterialFunctionCall(Float2=NamedRerouteUsage()) {materialFunction=&BreakOutFloat2Components}
  n46 = VectorParameter() {group=Opacity, parameterName=Opacity_EdgeMask_InvertControl, channelNames {a="", b=Power, g=Multiply, r=Offset}, defaultValue=<rgba 0 3 3 1>}
  n47 = VectorParameter() {group=Opacity, parameterName=Opacity_EdgeMask_Control, channelNames {a="", b=Power, g=Multiply, r=Offset}, defaultValue=<rgba 0 1000 1000 1>}
  n48 = Multiply(
   A=Multiply(
    A=Multiply(A=@n43, B=@n44)
    B=MaterialFunctionCall!(UorV=@n45.R, Invert_EdgeControl=@n46, NoInvert_EdgeControl=@n47) {materialFunction=/Game/Effects/Niagara/MasterMaterials/Functions/Common/FXCommon_EdgeMask_Ex}.Mask
   )
   B=MaterialFunctionCall!(ObjectOrParticlePosition=ParticlePositionWS()) {materialFunction=/Game/Effects/Fort_Effects/Materials/Toon/CustomDepthOutLine/MF_BackfaceCulling}.Result
  )
  Multiply(A=ParticleColor().A, B=Multiply(A=Multiply(A=@n42, B=@n48), B=NamedRerouteUsage()))
  n49 = MaterialFunctionCall(Float2=TextureCoordinate()) {materialFunction=&BreakOutFloat2Components}
  n50 = VectorParameter() {parameterName=UVControl_ScaleUVByCenter, defaultValue=<rgba 1 1.4 0 1>}
  NamedRerouteDeclaration(None=MaterialFunctionCall(
    UVs=MaterialFunctionCall(X=@n49.G, Y=@n49.G) {materialFunction=&MakeFloat2}.Result
    Texture Scale=AppendVector(A=@n50.R, B=@n50.R)
   ) {materialFunction=&ScaleUVsByCenter}.UVs) {name=UV, nodeColor=<rgba 0.8516 0 1 1>}
  n51 = VertexColor()
  n52 = ScalarParameter() {defaultValue=114.514, parameterName=Random}
  NamedRerouteDeclaration(None=Add(A=Multiply(A=@n51.B, B=@n52), B=Time() {bOverride_Period=true, period=9999})) {name=Time, nodeColor=<rgba 0 0.7187 1 1>}
  n53 = NamedRerouteUsage()
  n54 = VectorParameter() {group=32.TextureTex_Trd, parameterName=TextureTex_Trd_UVControl_ExtraLayer, defaultValue=<rgba 0.015 0.015 15 15>}
  n55 = Transform(None=Constant3Vector() {constant=<rgba 0 0 -1 0>}) {transformSourceType=TRANSFORMSOURCE_Camera}
  n56 = Transform(None=Constant3Vector() {constant=<rgba 1 0 0 0>}) {transformSourceType=TRANSFORMSOURCE_Camera}
  n57 = ObjectPositionWS()
  n58 = NamedRerouteDeclaration(None=MaterialFunctionCall(
    UVs=MaterialFunctionCall(Plane Normal Axis=@n55, Plane X Axis=@n56, Plane Center=@n57) {materialFunction=&VirtualPlaneCoordinates}.UVs
    Texture Scale=ScalarParameter() {defaultValue=1, group=00.Base, parameterName=Base_TextureTexUV}
   ) {materialFunction=&ScaleUVsByCenter}.UVs) {name=ScreenUV, nodeColor=<rgba 0 0.3203 1 1>}
  n59 = TextureObjectParameter() {
   group=32.TextureTex_Trd
   parameterName=TextureTex_Trd_Tex
   samplerType=SAMPLERTYPE_LinearGrayscale
   texture=/Game/Global/Textures/Default/Noise/T_WhiteNoise
  }
  NamedRerouteDeclaration(None=TextureSample!(
    UVs=MaterialFunctionCall!(UVInput=@n58, TimeInput=@n53, Control_RGB=@n54.RGB, Control_A=@n54.RGB) {materialFunction=/Game/Effects/Niagara/MasterMaterials/Functions/Common/FXCommon_QuickUVControl}.UV
    Tex=@n59
    Level
    Apply View MipBias
   ) {constMipValue=0, mipValueMode=TMVM_MipLevel, samplerType=SAMPLERTYPE_LinearColor}.R) {name=TextureTex_Trd_ExtraLayer, nodeColor=<rgba 0.3125 1 0 1>}
  n60 = VectorParameter() {group=32.TextureTex_Trd, parameterName=TextureTex_Trd_UVControl, defaultValue=<rgba 0.013 0.013 10 10>}
  NamedRerouteDeclaration(None=TextureSample(
    UVs=MaterialFunctionCall(UVInput=@n58, TimeInput=@n53, Control_RGB=@n60.RGB, Control_A=@n60.RGB) {materialFunction=&FXCommon_QuickUVControl}.UV
    Tex=@n59
   ) {constMipValue=0, mipValueMode=TMVM_MipLevel, samplerType=SAMPLERTYPE_LinearColor}.R) {name=TextureTex_Trd, nodeColor=<rgba 0.3125 1 0 1>}
  n61 = MaterialFunctionCall(Float3=SceneTexture() {sceneTextureId=PPI_CustomDepth}.Color) {materialFunction=&BreakOutFloat3Components}
  n62 = PixelDepth()
  n63 = ComponentMask(None=@n62.R) {r=true}
  n64 = Saturate(None=Divide(A=Subtract(A=@n61.R, B=@n63), B=ScalarParameter() {defaultValue=50, group=05.CustomDepthCullAndFade, parameterName=CustomDepthFadeDistance}) {constB=50})
  n65 = Constant() {r=1}
  n66 = If(A=@n62.R, B=@n61.R, "A > B"=Constant(), "A == B"=@n65, "A < B"=@n65) {equalsThreshold=0.1}
  n67 = Constant() {r=1}
  NamedRerouteDeclaration(None=ShadingPathSwitch(Default=Multiply(A=@n64, B=@n66), Forward=@n67, Mobile=@n67)) {name=R_CustomDepthCullAndFade, nodeColor=<rgba 0 0.6953 1 1>}
  n68 = DynamicParameter() {paramNames [ScaleX,Age,Width,Param4]}
  Reroute(None=Add(A=@n37.WPO, B=Multiply(A=@n41.WorldPositionOffset, B=@n68.ScaleX))) {desc=WPO}
  StaticBool() {value=true}
  TextureCoordinate()
 ]
}
