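# Master translucent unlit material for stylized toon clouds on Niagara mesh particles in a glider's effects.
# Panning cloud height and normal maps drive masked vertex offset and fake banded lighting from the sky atmosphere light or an override direction.
# Camera-plane noise textures add flicker, hue shift and desaturation, while Fresnel and backface culling shape opacity.
# Dynamic parameters feed scale, age and two foot positions, whose sphere masks dent and tint the cloud.
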Material /BRCosmetics/Gadgets/Assets/VinderTech_GliderChute/Glider_Ensemble_Maroon_Male/FX/Materials/M_ToonCloud_Master
graph {
 properties {
  allowTranslucentCustomDepthWrites=true
  bUseTranslucencyVertexFog=false
  bUsedWithNiagaraMeshParticles=true
  blendMode=BLEND_Translucent
  pixelDepthOffsetMode=PDOM_Legacy
  shadingModel=MSM_Unlit
  usedShadingModels=Unlit
 }
 nodes [
  n1 = ParticleColor()
  n2 = MaterialFunctionCall(Hue Shift Percentage=@n1.B, Texture=NamedRerouteUsage()) {materialFunction=&HueShift}
  MaterialFunctionCall(Emissive=Multiply(A=Desaturation(None=@n2.Result, Fraction=@n1.R), B=@n1.G)) {materialFunction=&MF_FX_EmissiveAdjustments}
  MaterialFunctionCall(
   In=DepthFade(Opacity=Constant() {r=1}, FadeDistance=ScalarParameter() {parameterName=DepthFadeDistance})
   Contrast=ScalarParameter() {defaultValue=1, parameterName=DepthFadeContrast}
  ) {materialFunction=&CheapContrast}
  n3 = Time() {bOverride_Period=true, period=9999}
  n4 = DynamicParameter() {paramNames [ScaleX,Age,Param3,Param4]}
  NamedRerouteDeclaration(None=Multiply(A=Add(A=@n3, B=@n4.Age), B=ScalarParameter() {defaultValue=1, group=00.Base, parameterName=Base_Time_CloudTex_Scale})) {name=Base_CloudTexTime, nodeColor=<rgba 1 0.5783 0 1>}
  n5 = MaterialFunctionCall(Float2=TextureCoordinate()) {materialFunction=&BreakOutFloat2Components}
  n6 = ScalarParameter() {defaultValue=1, group=00.Base, parameterName=Base_UV_CloudTex_VPower}
  NamedRerouteDeclaration(None=MaterialFunctionCall(X=@n5.R, Y=Power(Base=@n5.G, Exp=@n6)) {materialFunction=&MakeFloat2}.Result) {name=Base_CloudTexUV, nodeColor=<rgba 0.0312 1 0 1>}
  n7 = Constant3Vector() {constant=<rgba 0 0 1 0>}
  n8 = NamedRerouteUsage()
  n9 = Multiply(A=NamedRerouteUsage(), B=NamedRerouteUsage())
  n10 = Multiply(
   A=Transform(None=LinearInterpolate(A=@n7, B=@n8, Alpha=@n9))
   B=ScalarParameter() {group=20.WPOBlend, parameterName=WPOBlend_Main_NormalStrength}
  )
  n11 = NamedRerouteUsage()
  n12 = ScalarParameter() {group=20.WPOBlend, parameterName=WPOBlend_Detail_Strength}
  n13 = NamedRerouteUsage()
  n14 = ScalarParameter() {group=20.WPOBlend, parameterName=WPOBlend_Detail_SpecialAxisStrength}
  n15 = VectorParameter() {group=20.WPOBlend, parameterName=WPOBlend_Detail_SpecialAxis, defaultValue=<rgba 0 0 0 1>}
  n16 = Add(
   A=Multiply(A=Multiply(A=@n11, B=@n12), B=VertexNormalWS())
   B=Multiply(A=Multiply(A=@n13, B=@n14), B=Transform(None=@n15.RGB) {transformSourceType=TRANSFORMSOURCE_Local})
  )
  n17 = Constant()
  n18 = NamedRerouteUsage()
  n19 = ScalarParameter() {defaultValue=1, group=20.WPOBlend, parameterName=WPOBlend_Main_Power}
  n20 = Multiply(
   A=Add(A=Power(Base=@n18, Exp=@n19), B=ScalarParameter() {defaultValue=-0.5, group=20.WPOBlend, parameterName=WPOBlend_Main_StrengthOffset}) {constB=-0.5}
   B=ScalarParameter() {defaultValue=100, group=20.WPOBlend, parameterName=WPOBlend_Main_Strength}
  )
  n21 = VertexNormalWS()
  n22 = Multiply(
   A=Add(A=QualitySwitch(Default=@n16, Low=@n17, High=@n16, Medium=@n16, Epic=@n16), B=Multiply(A=@n20, B=@n21))
   B=Multiply(A=NamedRerouteUsage(), B=NamedRerouteUsage())
  )
  n23 = Add(A=LinearInterpolate(A=Add(A=@n10, B=@n22)) {constB=0}, B=NamedRerouteUsage())
  n24 = DynamicParameter() {parameterIndex=1, defaultValue=<rgba 1 1 1 0>, paramNames [FeetPosX, FeetPosY, FeetPosZ, FeetHoleColor]}
  n25 = AppendVector(A=@n24.FeetPosX, B=@n24.FeetPosX)
  n26 = WorldPosition() {worldPositionShaderOffset=WPT_ExcludeAllShaderOffsets}
  n27 = VectorParameter() {group=70.Feet, parameterName=FeetHole_SizeAndHardness, defaultValue=<rgba 30 0.5 0 0>}
  n28 = DynamicParameter() {parameterIndex=2, paramNames [Feet2PosX,Feet2PosY,Feet2PosZ,Grow]}
  n29 = AppendVector(A=@n28.Feet2PosX, B=@n28.Feet2PosX)
  n30 = VectorParameter() {group=70.Feet, parameterName=FeetHole1_SizeAndHardness, defaultValue=<rgba 30 0.5 0 0>}
  n31 = Max(
   A=SphereMask(A=AppendVector(A=@n25, B=@n24.FeetPosZ), B=@n26.XYZ, Radius=@n27.R, Hardness=@n27.R)
   B=SphereMask(A=AppendVector(A=@n29, B=@n28.Feet2PosZ), B=@n26.XYZ, Radius=@n30.R, Hardness=@n30.R)
  )
  n32 = Constant3Vector() {constant=<rgba 0 0 -1 0>}
  n33 = Multiply(
   A=Multiply(A=@n31, B=Transform(None=@n32) {transformSourceType=TRANSFORMSOURCE_Local})
   B=ScalarParameter() {group=70.Feet, parameterName=Feet_WPOStrength}
  )
  NamedRerouteDeclaration(None=Multiply(A=Add(A=@n23, B=@n33), B=DynamicParameter() {paramNames [ScaleX,Age,Param3,Param4]}.ScaleX)) {name=Final_WPO, nodeColor=<rgba 0.2422 1 0 1>}
  n34 = Transform(None=Constant3Vector() {constant=<rgba 0 0 -1 0>}) {transformSourceType=TRANSFORMSOURCE_Camera}
  n35 = Transform(None=Constant3Vector() {constant=<rgba 1 0 0 0>}) {transformSourceType=TRANSFORMSOURCE_Camera}
  n36 = ParticlePositionWS()
  NamedRerouteDeclaration(None=MaterialFunctionCall(
    UVs=MaterialFunctionCall(Plane Normal Axis=@n34, Plane X Axis=@n35, Plane Center=@n36) {materialFunction=&VirtualPlaneCoordinates}.UVs
    Texture Scale=ScalarParameter() {defaultValue=1, group=00.Base, parameterName=Base_TextureTexUV}
   ) {materialFunction=&ScaleUVsByCenter}.UVs) {name=ScreenUV, nodeColor=<rgba 1 0.3984 0 1>}
  NamedRerouteDeclaration(None=Multiply(A=@n3, B=ScalarParameter() {defaultValue=1, group=00.Base, parameterName=Base_Time_TextureTex_Scale})) {name=Base_TextureTexTime, nodeColor=<rgba 0 1 0.6953 1>}
  n37 = NamedRerouteUsage()
  n38 = VectorParameter() {group=32.TextureTex_Trd, parameterName=TextureTex_Trd_UVControl_ExtraLayer, defaultValue=<rgba 1 1 0 0>}
  n39 = NamedRerouteUsage()
  n40 = TextureObjectParameter() {
   group=32.TextureTex_Trd
   parameterName=TextureTex_Trd_Tex
   samplerType=SAMPLERTYPE_LinearGrayscale
   texture=/Game/Global/Textures/Default/Noise/T_WhiteNoise
  }
  NamedRerouteDeclaration(None=TextureSample!(
    UVs=MaterialFunctionCall!(UVInput=@n39, TimeInput=@n37, Control_RGB=@n38.RGB, Control_A=@n38.RGB) {materialFunction=/Game/Effects/Niagara/MasterMaterials/Functions/Common/FXCommon_QuickUVControl}.UV
    Tex=@n40
    Level
    Apply View MipBias
   ) {constMipValue=0, mipValueMode=TMVM_MipLevel, samplerType=SAMPLERTYPE_LinearColor}.R) {name=TextureTex_Trd_ExtraLayer, nodeColor=<rgba 0.3125 1 0 1>}
  n41 = VectorParameter() {group=32.TextureTex_Trd, parameterName=TextureTex_Trd_UVControl, defaultValue=<rgba 1 1 0 0>}
  NamedRerouteDeclaration(None=TextureSample(
    UVs=MaterialFunctionCall(UVInput=@n39, TimeInput=@n37, Control_RGB=@n41.RGB, Control_A=@n41.RGB) {materialFunction=&FXCommon_QuickUVControl}.UV
    Tex=@n40
   ) {constMipValue=0, mipValueMode=TMVM_MipLevel, samplerType=SAMPLERTYPE_LinearColor}.R) {name=TextureTex_Trd, nodeColor=<rgba 0.3125 1 0 1>}
  n42 = TextureObjectParameter() {
   group=30.TextureTex_Main
   parameterName=TextureTex_Main_Tex
   samplerType=SAMPLERTYPE_LinearGrayscale
   texture=/BRCosmetics/Global/Textures/Default/Noise/T_SandTile
  }
  n43 = VectorParameter() {group=30.TextureTex_Main, parameterName=TextureTex_Main_UVControl, defaultValue=<rgba 1 1 0 0>}
  NamedRerouteDeclaration(None=MaterialFunctionCall(
    Texture=@n42
    UVControl_RGB=@n43.RGB
    UVControl_A=@n43.RGB
    "[Override]TimeInput"=NamedRerouteUsage()
    "[Override]UVInput"=NamedRerouteUsage()
   ) {materialFunction=&FXCommon_TextureSample}.R) {name=TextureTex_Main, nodeColor=<rgba 0 0.4141 1 1>}
  n44 = TextureObjectParameter() {
   group=31.TextureTex_Sec
   parameterName=TextureTex_Sec_Tex
   samplerType=SAMPLERTYPE_LinearGrayscale
   texture=&T_WhiteNoise
  }
  n45 = VectorParameter() {group=31.TextureTex_Sec, parameterName=TextureTex_Sec_UVControl, defaultValue=<rgba 1 1 0 0>}
  NamedRerouteDeclaration(None=MaterialFunctionCall(
    Texture=@n44
    UVControl_RGB=@n45.RGB
    UVControl_A=@n45.RGB
    "[Override]TimeInput"=NamedRerouteUsage()
    "[Override]UVInput"=NamedRerouteUsage()
   ) {materialFunction=&FXCommon_TextureSample}.R) {name=TextureTex_Sec, nodeColor=<rgba 0 0.4141 1 1>}
  n46 = MaterialFunctionCall(Float3=TransformPosition(None=WorldPosition().XYZ) {transformSourceType=TRANSFORMPOSSOURCE_World, transformType=TRANSFORMPOSSOURCE_Instance}) {materialFunction=&BreakOutFloat3Components}
  n47 = Multiply(A=@n46.B, B=ScalarParameter() {defaultValue=1, group=50.Tail, parameterName=Tail_ZAxisScale})
  n48 = DotProduct(
   A=VectorParameter() {group=50.Tail, parameterName=Tail_CenterAxis, defaultValue=<rgba 0 0 1 1>}.RGB
   B=Normalize(VectorInput=MaterialFunctionCall(X=@n46.R, Y=@n46.R, Z=@n47) {materialFunction=&MakeFloat3}.Result)
  )
  n49 = Saturate(None=Power(Base=@n48, Exp=ScalarParameter() {group=50.Tail, parameterName=Tail_ToShapeTexMask_Power}))
  n50 = Multiply(
   A=Power(Base=VertexInterpolator(VS=@n49).PS, Exp=ScalarParameter() {defaultValue=2, parameterName=Tail_Bright_Power})
   B=ScalarParameter() {defaultValue=0.5, parameterName=Tail_Bright}
  )
  n51 = ScalarParameter() {defaultValue=0.5, parameterName=FeetHole_Brightness}
  n52 = DynamicParameter() {parameterIndex=1, defaultValue=<rgba 1 1 1 0>, paramNames [FeetPosX, FeetPosY, FeetPosZ, FeetHoleColor]}
  n53 = LinearInterpolate(B=Multiply(A=OneMinus(None=@n31), B=Constant() {r=1}), Alpha=@n52.FeetHoleColor) {constA=1}
  n54 = Constant3Vector() {constant=<rgba 0 0 1 0>}
  n55 = LinearInterpolate(A=@n54, B=NamedRerouteUsage(), Alpha=NamedRerouteUsage())
  n56 = ScalarParameter() {group=40.Lighting_Base, parameterName=Lighting_DetailNormalFlat}
  n57 = LinearInterpolate(
   A=@n54
   B=MaterialFunctionCall(
    BaseNormal=NamedRerouteUsage()
    AdditionalNormal=MaterialFunctionCall(Normal=@n55, Flatness=@n56) {materialFunction=&FlattenNormal}.Result
   ) {materialFunction=&BlendAngleCorrectedNormals}.Result
   Alpha=Multiply(A=NamedRerouteUsage(), B=NamedRerouteUsage())
  )
  n58 = ScalarParameter() {group=41.Lighting_A, parameterName=Lighting_A_NormalFlat}
  n59 = DotProduct(
   A=NamedRerouteUsage()
   B=Transform(None=MaterialFunctionCall(Normal=@n57, Flatness=@n58) {materialFunction=&FlattenNormal}.Result)
  )
  n60 = CameraPositionWS()
  n61 = ParticlePositionWS()
  n62 = DotProduct(A=NamedRerouteUsage(), B=Normalize(VectorInput=Subtract(A=@n60, B=@n61)))
  n63 = ScalarParameter() {defaultValue=0.2, parameterName=FixShadowOffset_MaskOffset}
  n64 = Divide(A=Saturate(None=Subtract(A=@n62, B=@n63)), B=Subtract(B=@n63))
  n65 = LinearInterpolate(
   A=ScalarParameter() {group=41.Lighting_A, parameterName=Lighting_A_NDS_Offset}
   B=ScalarParameter() {group=41.Lighting_A, parameterName=Lighting_A_NDS_Offset_B}
   Alpha=Power(Base=Saturate(None=@n64), Exp=ScalarParameter() {defaultValue=1, parameterName=FixShadowOffset_Power})
  )
  n66 = ScalarParameter() {defaultValue=1, group=41.Lighting_A, parameterName=Lighting_A_NDS_Mul_Additional}
  n67 = ScalarParameter() {defaultValue=1, group=41.Lighting_A, parameterName=Lighting_A_NDS_Mul}
  n68 = NamedRerouteUsage()
  n69 = OneMinus(None=Power(Base=OneMinus(None=@n68)))
  n70 = Saturate(None=Multiply(A=Add(A=@n59, B=@n65), B=LinearInterpolate(A=@n66, B=@n67, Alpha=@n69)))
  n71 = ScalarParameter() {defaultValue=1, group=41.Lighting_A, parameterName=Lighting_A_NDS_Power}
  n72 = VectorParameter() {group=41.Lighting_A, parameterName=Lighting_A_NDS_Remap, channelNames {a="", b="", g=TH, r=TL}, defaultValue=<rgba 0 1 0 1>}
  n73 = MaterialFunctionCall(Input=Saturate(None=Power(Base=@n70, Exp=@n71)), Input Low=Constant(), Input High=Constant() {r=1}, Target Low=@n72.TL, Target High=@n72.TL) {materialFunction=&RemapValueRange}
  n74 = LinearInterpolate(
   A=VectorParameter() {group=40.Lighting_Base, parameterName=Lighting_ColorA, defaultValue=<rgba 0 0 0 1>}.RGB
   B=VectorParameter() {group=40.Lighting_Base, parameterName=Lighting_ColorB, defaultValue=<rgba 0 0 0 1>}.RGB
   Alpha=Multiply(A=LinearInterpolate(A=@n51, Alpha=@n53), B=Saturate(None=@n73.Result))
  )
  n75 = MaterialFunctionCall(Hue Shift Percentage=ScalarParameter() {group=41.Lighting_B, parameterName=Lighting_B_HueShift}, Texture=@n74) {materialFunction=&HueShift}
  n76 = ScalarParameter() {group=41.Lighting_B, parameterName=Lighting_B_Desaturation}
  n77 = MaterialFunctionCall(Normal=@n57, Flatness=ScalarParameter() {group=41.Lighting_B, parameterName=Lighting_B_NormalFlat}) {materialFunction=&FlattenNormal}
  n78 = Add(
   A=DotProduct(A=NamedRerouteUsage(), B=Transform(None=@n77.Result))
   B=ScalarParameter() {group=41.Lighting_B, parameterName=Lighting_B_NDS_Offset}
  )
  n79 = ScalarParameter() {defaultValue=1, group=41.Lighting_B, parameterName=Lighting_B_NDS_Mul}
  n80 = Power(
   Base=Saturate(None=Multiply(A=@n78, B=@n79))
   Exp=ScalarParameter() {defaultValue=1, group=41.Lighting_B, parameterName=Lighting_B_NDS_Power}
  )
  n81 = VectorParameter() {group=41.Lighting_B, parameterName=Lighting_B_NDS_Remap, channelNames {a="", b="", g=TH, r=TL}, defaultValue=<rgba 0 1 0 1>}
  n82 = NamedRerouteUsage()
  n83 = OneMinus(None=Power(Base=OneMinus(None=@n82)) {constExponent=1.25})
  n84 = Saturate(None=MaterialFunctionCall(
    Input=Saturate(None=@n80)
    Input Low=Constant()
    Input High=Constant() {r=1}
    Target Low=LinearInterpolate(A=@n81.B, B=@n81.B, Alpha=@n83)
    Target High=@n81.TH
   ) {materialFunction=&RemapValueRange}.Result)
  n85 = Desaturation(
   None=LinearInterpolate(A=Desaturation(None=@n75.Result, Fraction=@n76), B=@n74, Alpha=@n84)
   Fraction=LinearInterpolate(A=ScalarParameter() {defaultValue=0.5, parameterName=FeetHole_Desat}, Alpha=@n53) {constB=0}
  )
  NamedRerouteDeclaration(None=Multiply(A=Add(A=@n50), B=Multiply(A=@n85, B=@n84))) {name=Lighting_ColorOut, nodeColor=<rgba 0.5828 0 1 1>}
  n86 = NamedRerouteUsage()
  n87 = Normalize(VectorInput=Subtract(A=WorldPosition().XYZ, B=ParticlePositionWS()))
  n88 = Multiply(
   A=Add(A=DotProduct(A=@n86, B=@n87), B=ScalarParameter() {group=41.Lighting_C, parameterName=Lighting_C_NDS_Offset})
   B=ScalarParameter() {defaultValue=1, group=41.Lighting_C, parameterName=Lighting_C_NDS_Mul}
  )
  n89 = Power(
   Base=OneMinus(None=Saturate(None=@n88))
   Exp=ScalarParameter() {defaultValue=1, group=41.Lighting_C, parameterName=Lighting_C_NDS_PowerInvert}
  )
  n90 = Power(Base=Saturate(None=OneMinus(None=@n89)), Exp=ScalarParameter() {defaultValue=1, group=41.Lighting_C, parameterName=Lighting_C_NDS_Power})
  n91 = Constant()
  n92 = Constant() {r=1}
  n93 = VectorParameter() {group=41.Lighting_C, parameterName=Lighting_C_NDS_Remap, channelNames {a="", b="", g=TR, r=TL}, defaultValue=<rgba 0 1 0 1>}
  NamedRerouteDeclaration(None=Saturate(None=MaterialFunctionCall(Input=@n90, Input Low=@n91, Input High=@n92, Target Low=@n93.TL, Target High=@n93.TL) {materialFunction=&RemapValueRange}.Result)) {name=Lighting_C_Out, nodeColor=<rgba 1 0 0.2529 1>}
  n94 = Fresnel(ExponentIn=ScalarParameter() {group=60.Opacity, parameterName=Opacity_Fresnel_Exp}, BaseReflectFractionIn=Constant())
  n95 = TransformPosition(None=WorldPosition() {worldPositionShaderOffset=WPT_ExcludeAllShaderOffsets}.XYZ) {transformSourceType=TRANSFORMPOSSOURCE_World, transformType=TRANSFORMPOSSOURCE_Instance}
  n96 = Add(
   A=Divide(
    A=MaterialFunctionCall(Float3=@n95) {materialFunction=&BreakOutFloat3Components}.G
    B=ScalarParameter() {group=60.Opacity, parameterName=Opacity_FresnelMask_DivideLocalPos}
   ) {constB=2048}
   B=ScalarParameter() {group=60.Opacity, parameterName=Opacity_FresnelMask_Offset}
  )
  n97 = LinearInterpolate(
   A=Constant() {r=1}
   B=Power(Base=OneMinus(None=@n94), Exp=ScalarParameter() {group=60.Opacity, parameterName=Opacity_Fresnel_InvertPower})
   Alpha=OneMinus(None=Saturate(None=@n96))
  )
  n98 = MaterialFunctionCall!(ObjectOrParticlePosition=ParticlePositionWS()) {materialFunction=/Game/Effects/Fort_Effects/Materials/Toon/CustomDepthOutLine/MF_BackfaceCulling}
  NamedRerouteDeclaration(None=Multiply(A=Constant() {r=1}, B=Multiply(A=@n97, B=@n98.Result) {constA=1})) {name=Final_Opacity, nodeColor=<rgba 0 1 0.8828 1>}
  n99 = MaterialFunctionCall(Float2=TextureCoordinate()) {materialFunction=&BreakOutFloat2Components}
  n100 = Sine(None=Multiply(A=@n99.G) {constB=0.5})
  n101 = Power(Base=@n100, Exp=ScalarParameter() {defaultValue=2, group=20.Mask, parameterName=Mask_ShapeTex_A_UpPower})
  n102 = ScalarParameter() {defaultValue=5, group=20.Mask, parameterName=Mask_ShapeTex_A_DownPower}
  n103 = Saturate(None=Multiply(A=Subtract(A=@n99.G) {constB=0.475}) {constB=20})
  NamedRerouteDeclaration(None=LinearInterpolate(A=@n101, B=Power(Base=@n100, Exp=@n102), Alpha=@n103)) {name=Mask_ShapeTex_A, nodeColor=<rgba 1 0.6094 0 1>}
  n104 = VectorParameter() {desc=42.Fresnel, group=42.Fresnel, parameterName=Fresnel_Color, defaultValue=<rgba 1 1 1 1>}
  n105 = Multiply(A=Fresnel(ExponentIn=@n104.A) {baseReflectFraction=0}, B=@n104.RGB)
  n106 = NamedRerouteUsage()
  n107 = Multiply(A=NamedRerouteUsage(), B=ScalarParameter() {group=35.TextureBlend, parameterName=Trd_ExtraLayerBlendSpeed})
  n108 = LinearInterpolate(A=NamedRerouteUsage(), B=NamedRerouteUsage(), Alpha=ConstantBiasScale(None=Sine(None=@n107)))
  n109 = OneMinus(None=@n108)
  n110 = ScalarParameter() {group=35.TextureBlend, parameterName=Flicker_TextureMul}
  n111 = OneMinus(None=MaterialFunctionCall(In=Multiply(A=@n109, B=@n110), Contrast=ScalarParameter() {group=35.TextureBlend, parameterName=Flicker_TextureContrast}) {materialFunction=&CheapContrast}.Result)
  n112 = ScalarParameter() {group=35.TextureBlend, parameterName=Flicker_TextureMul2}
  n113 = ScalarParameter() {group=35.TextureBlend, parameterName=Flicker_LowValue}
  n114 = Multiply(A=NamedRerouteUsage(), B=ScalarParameter() {group=35.TextureBlend, parameterName=Flicker_TimeSpeed})
  n115 = Power(Base=ConstantBiasScale(None=Sine(None=@n114)), Exp=ScalarParameter() {group=35.TextureBlend, parameterName=Flicker_TimeSlope})
  n116 = NamedRerouteUsage()
  n117 = MaterialFunctionCall(
   In=Multiply(A=OneMinus(None=@n116), B=ScalarParameter() {group=35.TextureBlend, parameterName=TextureSec_Mul})
   Contrast=ScalarParameter() {group=35.TextureBlend, parameterName=TextureSec_Contrast}
  ) {materialFunction=&CheapContrast}
  n118 = Multiply(
   A=Multiply(A=Multiply(A=@n111, B=@n112), B=LinearInterpolate(A=@n113, Alpha=@n115))
   B=MaterialFunctionCall(Input=@n117.Result, Target Low=ScalarParameter() {group=35.TextureBlend, parameterName=Flicker_TextureSecBlend_LowValue}) {materialFunction=&RemapValueRange}.Result
  )
  n119 = Add(A=Multiply(A=Multiply(A=@n106, B=@n118), B=VectorParameter() {group=35.TextureBlend, parameterName=Flicker_Color, defaultValue=<rgba 0 0 0 1>}.RGB))
  n120 = ScalarParameter() {group=35.TextureBlend, parameterName=CloudTex_Detail_ToHueSHift}
  n121 = Multiply(A=NamedRerouteUsage(), B=Multiply(A=NamedRerouteUsage(), B=NamedRerouteUsage()))
  n122 = OneMinus(None=@n117.Result)
  n123 = ScalarParameter() {group=35.TextureBlend, parameterName=TextureSec_ToHueShift}
  n124 = LinearInterpolate(
   A=@n108
   B=Power(Base=@n108, Exp=ScalarParameter() {group=35.TextureBlend, parameterName=TextureTrd_Power}) {constExponent=0.2}
   Alpha=@n117.Result
  )
  n125 = MaterialFunctionCall(
   In=Multiply(A=OneMinus(None=@n124), B=ScalarParameter() {group=35.TextureBlend, parameterName=TextureTrd_Mul})
   Contrast=ScalarParameter() {group=35.TextureBlend, parameterName=TextureTrd_Contrast}
  ) {materialFunction=&CheapContrast}
  n126 = Constant()
  n127 = Constant() {r=1}
  n128 = VectorParameter() {group=35.TextureBlend, parameterName=TextureTrd_Remap, channelNames {a="", b="", g=TH, r=TL}, defaultValue=<rgba 0 1 0 1>}
  n129 = NamedRerouteUsage()
  n130 = MaterialFunctionCall(
   In=Multiply(A=OneMinus(None=@n129), B=ScalarParameter() {group=35.TextureBlend, parameterName=TextureMain_Mul})
   Contrast=ScalarParameter() {group=35.TextureBlend, parameterName=TextureMain_Contrast}
  ) {materialFunction=&CheapContrast}
  n131 = Constant()
  n132 = Constant() {r=1}
  n133 = VectorParameter() {group=35.TextureBlend, parameterName=TextureMain_Remap, defaultValue=<rgba 0 1 0 1>}
  n134 = VectorParameter() {group=35.TextureBlend, parameterName=TextureSec_Remap, defaultValue=<rgba 0 1 0 1>}
  n135 = Multiply(
   A=Min(
    A=MaterialFunctionCall(Input=@n125.Result, Input Low=@n126, Input High=@n127, Target Low=@n128.TL, Target High=@n128.TL) {materialFunction=&RemapValueRange}.Result
    B=MaterialFunctionCall(Input=@n130.Result, Input Low=@n131, Input High=@n132, Target Low=@n133.R, Target High=@n133.R) {materialFunction=&RemapValueRange}.Result
   )
   B=MaterialFunctionCall(Input=@n117.Result, Input Low=Constant(), Input High=Constant() {r=1}, Target Low=@n134.R, Target High=@n134.R) {materialFunction=&RemapValueRange}.Result
  )
  n136 = NamedRerouteUsage()
  n137 = Multiply(A=@n121, B=ScalarParameter() {group=35.TextureBlend, parameterName=CloudTex_Detail_ToDesat})
  n138 = Multiply(A=@n122, B=ScalarParameter() {group=35.TextureBlend, parameterName=TextureSec_ToDesat})
  n139 = MaterialFunctionCall(
   Hue Shift Percentage=Add(A=Multiply(A=@n121, B=@n120), B=Multiply(A=@n122, B=@n123))
   Texture=Desaturation(None=Multiply(A=@n135, B=@n136), Fraction=Add(A=@n137, B=@n138))
  ) {materialFunction=&HueShift}
  NamedRerouteDeclaration(None=Multiply(A=Add(A=@n105), B=Multiply(A=@n119, B=@n139.Result))) {name=Final_Emissive, nodeColor=<rgba 0 1 0.0391 1>}
  n140 = MaterialFunctionCall(Float2=TextureCoordinate()) {materialFunction=&BreakOutFloat2Components}
  n141 = Sine(None=Multiply(A=@n140.G) {constB=0.5})
  n142 = ScalarParameter() {defaultValue=2, group=80.Additional, parameterName=Additional_Mask_ShapeTex_C_UpPower}
  n143 = Multiply(A=Subtract(A=@n140.G) {constB=0.475}) {constB=20}
  NamedRerouteDeclaration(None=LinearInterpolate(A=Power(Base=@n141, Exp=@n142), Alpha=Saturate(None=@n143))) {name=Additional_Mask_ShapeTex_C, nodeColor=<rgba 1 0.6094 0 1>}
  NamedRerouteDeclaration(None=LinearInterpolate(A=@n101, Alpha=@n103)) {name=Additional_Mask_ShapeTex_A_JustUp, nodeColor=<rgba 1 0.6094 0 1>}
  n144 = VectorParameter() {parameterName=FELighting, defaultValue=<rgba 0 0 1 1>}
  NamedRerouteDeclaration(None=LinearInterpolate(
    A=SkyAtmosphereLightDirection()
    B=Normalize(VectorInput=@n144.RGB)
    Alpha=DynamicParameter() {parameterIndex=3, paramNames [Param1,FE,Param3,Param4]}.FE
   )) {name=LightDir, nodeColor=<rgba 1 0 0.7734 1>}
  n145 = MaterialFunctionCall(Float2=TextureCoordinate()) {materialFunction=&BreakOutFloat2Components}
  n146 = Sine(None=Multiply(A=@n145.G) {constB=0.5})
  n147 = ScalarParameter() {defaultValue=2, group=20.Mask, parameterName=Mask_ShapeTex_B_UpPower}
  n148 = ScalarParameter() {defaultValue=5, group=20.Mask, parameterName=Mask_ShapeTex_B_DownPower}
  n149 = Multiply(A=Subtract(A=@n145.G) {constB=0.475}) {constB=20}
  NamedRerouteDeclaration(None=LinearInterpolate(A=Power(Base=@n146, Exp=@n147), B=Power(Base=@n146, Exp=@n148), Alpha=Saturate(None=@n149))) {name=Mask_ShapeTex_B, nodeColor=<rgba 1 0.2405 0 1>}
  NamedRerouteDeclaration(None=LinearInterpolate(A=ScalarParameter() {group=50.Tail, parameterName=Tail_ToShapeTexMask_LowestValue}, Alpha=OneMinus(None=@n49))) {name=Tail_ToShapeTexMask, nodeColor=<rgba 0.7109 0 1 1>}
  n150 = Multiply(A=NamedRerouteUsage(), B=ScalarParameter() {group=50.Tail, parameterName=Tail_WaveWPO_Speed})
  n151 = Power(Base=@n48, Exp=ScalarParameter() {group=50.Tail, parameterName=Tail_WaveWPO_MaskPower})
  n152 = Multiply(
   A=Sine(None=Add(A=@n49, B=@n150))
   B=Multiply(A=Saturate(None=@n151), B=ScalarParameter() {group=50.Tail, parameterName=Tail_WaveWPO_Strength})
  )
  n153 = VertexNormalWS()
  n154 = ScalarParameter() {group=50.Tail, parameterName=Tail_VertWPO_Mask_Power}
  n155 = Multiply(A=Saturate(None=Power(Base=@n48, Exp=@n154)), B=ScalarParameter() {group=50.Tail, parameterName=Tail_VertWPO_Strength})
  n156 = VertexNormalWS()
  NamedRerouteDeclaration(None=Add(A=Multiply(A=@n152, B=@n153), B=Multiply(A=@n155, B=@n156))) {name=Tail_AdditionalWPO, nodeColor=<rgba 0 0.2969 1 1>}
  n157 = TextureObjectParameter() {
   group=11.CloudTex_Detail
   parameterName=CloudTex_Detail_NormalTex
   samplerType=SAMPLERTYPE_Normal
   texture=/Game/Effects/Fort_Effects/Gameplay/Floppers/FlopperUpdrade/T_FlowSandTile_N
  }
  n158 = VectorParameter() {group=11.CloudTex_Detail, parameterName=CloudTex_Detail_UVControl, defaultValue=<rgba 1 1 0 0>}
  n159 = NamedRerouteUsage()
  n160 = NamedRerouteUsage()
  NamedRerouteDeclaration(None=MaterialFunctionCall(Texture=@n157, UVControl_RGB=@n158.RGB, UVControl_A=@n158.RGB, "[Override]TimeInput"=@n159, "[Override]UVInput"=@n160) {materialFunction=&FXCommon_TextureSample}.RGB) {name=CloudTex_Detail_Normal, nodeColor=<rgba 0.5938 0 1 1>}
  n161 = TextureObjectParameter() {
   group=11.CloudTex_Detail
   parameterName=CloudTex_Detail_HeightTex
   samplerType=SAMPLERTYPE_LinearGrayscale
   texture=/Game/Effects/Fort_Effects/Gameplay/Floppers/FlopperUpdrade/T_FlowSandTile
  }
  NamedRerouteDeclaration(None=MaterialFunctionCall(Texture=@n161, UVControl_RGB=@n158.RGB, UVControl_A=@n158.RGB, "[Override]TimeInput"=@n159, "[Override]UVInput"=@n160) {materialFunction=&FXCommon_TextureSample}.R) {name=CloudTex_Detail_Height, nodeColor=<rgba 0 0.3437 1 1>}
  n162 = TextureObjectParameter() {
   group=10.CloudTex_Main
   parameterName=CloudTex_Main_HeightTex
   samplerType=SAMPLERTYPE_LinearGrayscale
   texture=/Game/Gadgets/Assets/VinderTech_GliderChute/Glider_Thrive_Female/FX/Texture/T_CloudBubble
  }
  n163 = VectorParameter() {group=10.CloudTex_Main, parameterName=CloudTex_Main_UVControl, defaultValue=<rgba 1 1 0 0>}
  n164 = NamedRerouteUsage()
  n165 = NamedRerouteUsage()
  NamedRerouteDeclaration(None=MaterialFunctionCall(Texture=@n162, UVControl_RGB=@n163.RGB, UVControl_A=@n163.RGB, "[Override]TimeInput"=@n164, "[Override]UVInput"=@n165) {materialFunction=&FXCommon_TextureSample}.R) {name=CloudTex_Main_Height, nodeColor=<rgba 0 0.3437 1 1>}
  n166 = TextureObjectParameter() {
   group=10.CloudTex_Main
   parameterName=CloudTex_Main_NormalTex
   samplerType=SAMPLERTYPE_Normal
   texture=/Game/Gadgets/Assets/VinderTech_GliderChute/Glider_Thrive_Female/FX/Texture/T_CloudBubble_N
  }
  NamedRerouteDeclaration(None=MaterialFunctionCall(Texture=@n166, UVControl_RGB=@n163.RGB, UVControl_A=@n163.RGB, "[Override]TimeInput"=@n164, "[Override]UVInput"=@n165) {materialFunction=&FXCommon_TextureSample}.RGB) {name=CloudTex_Main_Normal, nodeColor=<rgba 1 0.3757 0.6432 1>}
  NamedRerouteUsage()
  NamedRerouteUsage()
  NamedRerouteUsage()
 ]
}
