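# Main Niagara effect for the Ensemble Maroon male glider, pairing a cloud mesh, trailing deforming meshes and sprite puffs.
# Cloud003 scales in with an overshoot and feeds two feet positions to its material.
# FollowingDeform drives trail stretching; sprites read it for velocity direction, with spawn rate following a smoothed system value.
# Curves run on system age, User.Deployed adds 100 to skip the intro, and IsFrontEnd switches offsets, scale and tint.
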NiagaraSystem /BRCosmetics/Gadgets/Assets/VinderTech_GliderChute/Glider_Ensemble_Maroon_Male/FX/NS_EM_GliderMain
dependencies {
 usedDataInterfaces [&NiagaraDataInterfaceCurve, &NiagaraDataInterfacePlatformSet, &NiagaraDataInterfaceParticleRead]
 usedDynamicInputs []
 usedModules [&SystemState, &InterpolateOverTime]
}
emitters {
 Cloud003 {
  emitterUpdateScript [
   &EmitterState(
    Inactive Response=::NewEnumerator1
    Life Cycle Mode=::NewEnumerator1
    Loop Behavior=::NewEnumerator1
    Loop Delay=1.1
    Loop Duration Mode=::NewEnumerator1
    UseLoopDelay=false
   )
   &SpawnBurst_Instantaneous(Use Spawn Probability=false, UseLoopCountLimit=false)
  ]
  particleSpawnScript [
   &InitializeParticle(
    Color Mode=::NewEnumerator0
    Lifetime=5
    Mesh Scale=&Multiply_VectorByFloat(Vector=<0.47 0.45 0.463>, Float=0.36)
    Mesh Scale Mode=::NewEnumerator4
    Position Mode=::NewEnumerator1
    Position Offset=<0 -10 -22>
    Position Offset Coordinate Space=::Local
    UsePositionOffset=true
    Write Material Random=true
   )
   &InitialMeshOrientation(
    Mesh Axis To Orient=<0 1 0>
    Mesh Orientation Mode=::NewEnumerator2
    Orientation Coordinate Space=::Local
    Orientation Vector=&NormalizeVector(Vector To Normalize=<1 0 0>, Safe Normalize=true, Safe Fallback Threshold=0, Safe Fallback Vector=<0 0 0.001>)
    Use Rotation Vector=false
   )
   /SetVariables(Particles.MeshOrientation=&MultiplyQuaternion(
     Quaternion A=&MakeQuatFromAxisAngle(Axis=<1 0 0>, AngleInDegrees=-3.5)
     Quaternion B=$Particles.MeshOrientation
    )) {isSetParameters=true}
  ]
  particleUpdateScript [
   &ParticleState(Kill Particles When Lifetime Has Elapsed=false, Let Infinitely Lived Particles Die When Emitter Deactivates=true)
   &ScaleMeshSize(Scale Factor=&VectorFromFloat(Value=&MakeCustomFloatFromBool(Bool=$User.IsFrontEnd, True Float=0.9, False Float=1)))
   /SetVariables(Particles.FeetPos:Vector3f=/Niagara/DynamicInputs/Transforms/TransformPosition(
     Position=&Add_Vector(
      A=Linked() {linkedVariable {name=Particles.Position, type=NiagaraPosition}}
      B=&MakeCustomVectorFromBool(Bool=$User.IsFrontEnd, True Vector=<0 -20 1500>, False Vector=<0 -20 15>)
     )
     Source Space=::Local
     Destination Space=::World
    )) {isSetParameters=true}
   /SetVariables(Particles.Feet2Pos:Vector3f=&TransformPosition(
     Position=&Add_Vector(
      A=Linked() {linkedVariable {name=Particles.Position, type=NiagaraPosition}}
      B=&MakeCustomVectorFromBool(Bool=$User.IsFrontEnd, True Vector=<0 35 1500>, False Vector=<0 35 15>)
     )
     Source Space=::Local
     Destination Space=::World
    )) {isSetParameters=true}
   &FN_Timer(Increment Multiplier:NiagaraFloat=&Multiply_Float(A=&ClampFloat(Float=$System.InterpolateOverTime.InterpolateOverTime_Float.Moving Average), B=0.5))
   &ScaleMeshSize(Scale Factor=&VectorFromFloat(Value=&FloatFromCurve(
      FloatCurve=DataInterface() {propertyValues {Curve {keys [
          {arriveTangent=4.1785, interpMode=RCIM_Cubic, leaveTangent=0.6039, tangentMode=RCTM_Break, time=0.4, value=0}
          {arriveTangent=0.584, interpMode=RCIM_Cubic, leaveTangent=0.584, tangentMode=RCTM_User, time=1.6969, value=0.47}
          {interpMode=RCIM_Cubic, tangentMode=RCTM_Break, time=2.003, value=1.0936}
          {arriveTangent=-0.1623, interpMode=RCIM_Cubic, leaveTangent=-0.1623, time=2.1898, value=0.96}
          {arriveTangent=0.0512, interpMode=RCIM_Cubic, leaveTangent=0.0512, time=2.3997, value=1.0292}
          {arriveTangent=-0.0534, interpMode=RCIM_Cubic, leaveTangent=-0.0534, time=2.6408, value=0.9831}
          <at 2.9469 1 RCIM_Cubic>
         ]}}}
      CurveIndex=&Add_Float(A=$System.Age, B=&Multiply_FloatByInt(Float=100, Integer=$User.Deployed))
     ))) {module=ScaleMeshSize001}
   &DynamicMaterialParameters(
    Index 0 Param 1=&Multiply_Float(A=&MakeFloatFromVector(Vector=$Particles.Scale, Channel=::NewEnumerator0), B=2)
    Index 0 Param 2=$Particles.FN_Timer.Value
    Index 0 Param 4=&FloatFromCurve(
     FloatCurve=DataInterface() {propertyValues {Curve {keys [
         {interpMode=RCIM_Cubic, leaveTangent=4.4102, leaveTangentWeight=0.1524, tangentMode=RCTM_Break, time=0, value=0}
         {arriveTangent=0.4089, interpMode=RCIM_Cubic, leaveTangent=0.4089, tangentMode=RCTM_User, time=0.572, value=1.375}
         {tangentMode=RCTM_User, time=1.3, value=1}
        ]}}}
     CurveIndex=$Particles.Age
     Scale Curve=&MakeCustomFloatFromBool(Bool=$User.IsFrontEnd, True Float=1.1, False Float=1)
    )
    Index 1 Param 1=&MakeFloatFromVector(Vector=$Particles.FeetPos, Channel=::NewEnumerator0)
    Index 1 Param 2=&MakeFloatFromVector(Vector=$Particles.FeetPos, Channel=::NewEnumerator1)
    Index 1 Param 3=&MakeFloatFromVector(Vector=$Particles.FeetPos, Channel=::NewEnumerator2)
    Index 1 Param 4=&FloatFromCurve(
     FloatCurve=DataInterface() {propertyValues {Curve {keys [<at 0 0>,<at 2 0>,<at 3 1>]}}}
     CurveIndex=&ClampFloat(
      Float=&Add_Float(A=$System.Age, B=&Multiply_FloatByInt(Float=100, Integer=$User.Deployed))
      Max=5
     )
    )
    Index 2 Param 1=&MakeFloatFromVector(Vector=$Particles.Feet2Pos, Channel=::NewEnumerator0)
    Index 2 Param 2=&MakeFloatFromVector(Vector=$Particles.Feet2Pos, Channel=::NewEnumerator1)
    Index 2 Param 3=&MakeFloatFromVector(Vector=$Particles.Feet2Pos, Channel=::NewEnumerator2)
    Index 2 Param 4=&MakeCustomFloatFromBool(
     Bool=$User.IsFrontEnd
     True Float=&FloatFromCurve(
      FloatCurve=DataInterface() {propertyValues {Curve {keys [
          {arriveTangent=-0.3111, interpMode=RCIM_Cubic, leaveTangent=-0.3111, tangentMode=RCTM_User, time=0, value=1}
          {arriveTangent=-0.4837, interpMode=RCIM_Cubic, leaveTangent=-0.4837, tangentMode=RCTM_User, time=0.9, value=0.65}
          {arriveTangent=-0.2194, interpMode=RCIM_Cubic, leaveTangent=-0.2194, tangentMode=RCTM_User, time=1.0858, value=-0.3}
          {arriveTangent=0.0139, interpMode=RCIM_Cubic, leaveTangent=0.0139, tangentMode=RCTM_User, time=1.3, value=0.15}
          {arriveTangent=0.0474, interpMode=RCIM_Cubic, leaveTangent=0.0474, tangentMode=RCTM_User, time=1.475, value=-0.13}
          {interpMode=RCIM_Cubic, leaveTangent=0.1333, time=1.7, value=0.07}
          <at 2 -0.08 RCIM_Cubic>
         ]}}}
      CurveIndex=&Add_Float(A=$System.Age, B=&Multiply_FloatByInt(Float=100, Integer=$User.Deployed))
     )
     False Float=&FloatFromCurve(
      FloatCurve=DataInterface() {propertyValues {Curve {keys [
          {arriveTangent=-0.3111, interpMode=RCIM_Cubic, leaveTangent=-0.3111, tangentMode=RCTM_User, time=0, value=1}
          {arriveTangent=-0.4837, interpMode=RCIM_Cubic, leaveTangent=-0.4837, tangentMode=RCTM_User, time=0.9, value=0.85}
          {arriveTangent=-0.2194, interpMode=RCIM_Cubic, leaveTangent=-0.2194, tangentMode=RCTM_User, time=1.0858, value=-0.3}
          {arriveTangent=0.0139, interpMode=RCIM_Cubic, leaveTangent=0.0139, tangentMode=RCTM_User, time=1.3, value=0.15}
          {arriveTangent=0.0474, interpMode=RCIM_Cubic, leaveTangent=0.0474, tangentMode=RCTM_User, time=1.475, value=-0.13}
          {interpMode=RCIM_Cubic, leaveTangent=0.1333, time=1.7, value=0.07}
          <at 2 -0.08 RCIM_Cubic>
         ]}}}
      CurveIndex=&Add_Float(A=$System.Age, B=&Multiply_FloatByInt(Float=100, Integer=$User.Deployed))
     )
    )
    Index 3 Param 1=&MakeCustomFloatFromBool(
     Bool=$User.IsFrontEnd
     True Float=&FloatFromCurve(
      FloatCurve=DataInterface() {propertyValues {Curve {keys [<at 0 3>,<at 0.9 2>,<at 1.5 0>]}}}
      CurveIndex=&Add_Float(A=$System.Age, B=&Multiply_FloatByInt(Float=100, Integer=$User.Deployed))
      Scale Curve=0.04
     )
     False Float=&FloatFromCurve(
      FloatCurve=DataInterface() {propertyValues {Curve {keys [<at 0 1>,<at 1.5 0>]}}}
      CurveIndex=&Add_Float(A=$System.Age, B=&Multiply_FloatByInt(Float=100, Integer=$User.Deployed))
      Scale Curve=0.05
     )
    )
    Index 3 Param 2=&MakeCustomFloatFromBool(Bool=$User.IsFrontEnd, True Float=1, False Float=0)
    Index 3 Param 4=$Particles.FN_Timer.Value
    Module.Param0WriteEnabled=true
    Module.Param0WriteEnabled1=true
    Module.Param0WriteEnabled2=true
    Module.Param0WriteEnabled3=true
    Module.Param1WriteEnabled=true
    Module.Param1WriteEnabled1=true
    Module.Param1WriteEnabled2=true
    Module.Param1WriteEnabled3=true
    Module.Param2WriteEnabled=true
    Module.Param2WriteEnabled1=true
    Module.Param2WriteEnabled2=true
    Module.Param2WriteEnabled3=true
    Module.Param3WriteEnabled=true
    Module.Param3WriteEnabled1=true
    Module.Param3WriteEnabled2=true
    Module.Param3WriteEnabled3=true
    Write Parameter Index 1=true
    Write Parameter Index 2=true
    Write Parameter Index 3=true
   )
   &ScaleColor(
    Scale RGB=&VectorFromFloat(Value=&MakeCustomFloatFromBool(Bool=$User.IsFrontEnd, True Float=1, False Float=1))
    ScaleA=true
    ScaleRGB=true
   ) {enabled=false}
   &ScaleColor(
    Scale RGB=&MakeVector(
     X=&MakeCustomFloatFromBool(Bool=$User.IsFrontEnd, True Float=0, False Float=0)
     Y=&MakeCustomFloatFromBool(Bool=$User.IsFrontEnd, True Float=1.6, False Float=1)
     Z=&MakeCustomFloatFromBool(Bool=$User.IsFrontEnd, True Float=0, False Float=0)
    )
    ScaleA=true
    ScaleRGB=true
   ) {module=ScaleColor001}
  ]
  rendererClasses [&NiagaraMeshRendererProperties]
  renderers [
   {rendererClass=&NiagaraMeshRendererProperties, rendererIndex=0}
   {rendererClass=&NiagaraMeshRendererProperties, rendererIndex=1}
  ]
 }
 Empty {
  emitterUpdateScript [
   &EmitterState()
   &SpawnRate(
    SpawnRate=&FloatFromCurve(
     FloatCurve=DataInterface() {propertyValues {Curve {keys [<at 1.5 0>,<at 1.8 1>]}}}
     CurveIndex=&Add_Float(A=$System.Age, B=&Multiply_FloatByInt(Float=100, Integer=$User.Deployed))
     Scale Curve=&Lerp_Float(A=0, B=10, Alpha=$System.InterpolateOverTime.InterpolateOverTime_Float.Moving Average)
    )
    Use Spawn Probability=false
   )
  ]
  particleSpawnScript [
   &InitializeParticle(
    Color=<rgba 1 1 1 0.3>
    Lifetime=0.5
    Position Offset=<0 10 -50>
    Sprite Rotation Mode=::NewEnumerator2
    Sprite Size Mode=::NewEnumerator1
    Uniform Sprite Size Max=150
    Uniform Sprite Size Min=50
    UsePositionOffset=true
    Write Material Random=true
   )
   &SubUVAnimation(End Frame=4, SubUV Animation Mode:ENiagara_SubUVLookupMode=::NewEnumerator2)
   &ShapeLocation(MaskSpawnGroup=false, Override Seed=false, Sphere Radius=25, UseOwnerScale=true)
   &self:ReadFollowing(New Particle Attribute Reader=DataInterface() {propertyValues {EmitterBinding {bindingMode=Other, emitterName=FN_FollowingDeformTemplate001}}}) {module=ReadFollowing}
   &AddVelocity(Velocity=&Multiply_VectorByFloat(
     Vector=&TransformVector(
      Vector=&NormalizeVector(
       Vector To Normalize=&Subtract_Vector(A=$Particles.ReadFollowing.Value, B=Linked() {linkedVariable {name=Engine.Owner.Position, type=NiagaraPosition}})
       Safe Normalize=true
       Safe Fallback Threshold=0
       Safe Fallback Vector=<0 0 0.001>
      )
      Source Space=::World
      Destination Space=::Local
     )
     Float=250
    ))
   &ApplyInitialForces()
  ]
  particleUpdateScript [
   &ParticleState()
   &DynamicMaterialParameters(
    Index 0 Param 1=&FloatFromCurve(FloatCurve=DataInterface() {propertyValues {Curve {keys [<at 0 -1>,<at 1 1>]}}})
    Index 0 Param 2=&MakeCustomFloatFromBool(Bool=$User.IsFrontEnd, True Float=0, False Float=0)
    Index 0 Param 3=&MakeCustomFloatFromBool(Bool=$User.IsFrontEnd, True Float=1.6, False Float=1)
    Index 0 Param 4=&MakeCustomFloatFromBool(Bool=$User.IsFrontEnd, True Float=0, False Float=0)
    Module.Param0WriteEnabled=true
    Module.Param1WriteEnabled=true
    Module.Param2WriteEnabled=true
    Module.Param3WriteEnabled=true
   )
   &GravityForce() {enabled=false}
   &SolveForcesAndVelocity(Clamp Velocity=false, Limit Acceleration=false)
   &ScaleSpriteSize(Uniform Curve Sprite Scale=DataInterface() {propertyValues {Curve {keys [
        {arriveTangent=4.1785, interpMode=RCIM_Cubic, leaveTangent=4.1785, tangentMode=RCTM_User, time=0, value=0}
        {arriveTangent=0.9995, interpMode=RCIM_Cubic, leaveTangent=0.9995, tangentMode=RCTM_User, time=0.2223, value=0.6825}
        {arriveTangent=-0.0405, interpMode=RCIM_Cubic, leaveTangent=-0.0405, tangentMode=RCTM_User, time=1, value=1}
       ]}}})
   &ScaleColor(
    Scale Alpha=&FloatFromCurve(FloatCurve=DataInterface() {propertyValues {Curve {keys [
         {arriveTangent=24.7199, interpMode=RCIM_Cubic, leaveTangent=24.7199, tangentMode=RCTM_User, time=0, value=0}
         {arriveTangent=-0.1391, interpMode=RCIM_Cubic, leaveTangent=-0.1391, tangentMode=RCTM_User, time=0.1413, value=1.8879}
         {arriveTangent=-1.3479, interpMode=RCIM_Cubic, leaveTangent=-1.3479, time=0.3356, value=1.2904}
         {interpMode=RCIM_Cubic, tangentMode=RCTM_Break, time=0.8, value=1}
         {arriveTangent=-8.8798, interpMode=RCIM_Cubic, leaveTangent=-0.0405, tangentMode=RCTM_Break, time=1, value=0}
        ]}}})
    ScaleA=true
    ScaleRGB=true
   )
  ]
  rendererClasses [&NiagaraSpriteRendererProperties]
  renderers [{rendererClass=&NiagaraSpriteRendererProperties}]
 }
 FN_FollowingDeformTemplate001 {
  emitterUpdateScript [
   &EmitterState(
    Inactive Response=::NewEnumerator1
    Life Cycle Mode=::NewEnumerator1
    Loop Behavior=::NewEnumerator1
    Loop Duration Mode=::NewEnumerator1
    UseLoopDelay=false
   )
   &SpawnBurst_Instantaneous(Use Spawn Probability=false, UseLoopCountLimit=false)
  ]
  particleSpawnScript [&InitializeParticle(
    Color=<rgba 1 1 1 1.2>
    Lifetime=5
    Mesh Scale Mode=::NewEnumerator3
    Mesh Uniform Scale=0.4
    Position Mode=::NewEnumerator0
    Sprite Size Mode=::NewEnumerator3
    Uniform Sprite Size=100
    Write Material Random=true
   )]
  particleUpdateScript [
   &ParticleState(Kill Particles When Lifetime Has Elapsed=false, Let Infinitely Lived Particles Die When Emitter Deactivates=true)
   /Game/Effects/Niagara/Modules/FollowingDeform/FN_FollowingDeform(
    CustomMeshOrientation=false
    MaxStretchDistance=100
    MeshZ+AxisAlignLS:Vector3f=<0.01 0.01 1>
    PlaneConstraint=true
    RootPosition:NiagaraPosition=&TransformPosition(Position=<0 -35 -45>, Source Space=::Local, Destination Space=::World)
   )
   &DynamicMaterialParameters(
    Index 0 Param 1=&Multiply_Float(A=&MakeFloatFromVector(Vector=$Particles.Scale, Channel=::NewEnumerator0), B=1)
    Index 0 Param 3=&FloatFromCurve(
     FloatCurve=DataInterface() {propertyValues {Curve {keys [{interpMode=RCIM_Cubic, tangentMode=RCTM_Break, time=0, value=0}, <at 1.7 0.6>, <at 1.8 1>]}}}
     CurveIndex=&Add_Float(A=$System.Age, B=&Multiply_FloatByInt(Float=100, Integer=$User.Deployed))
    )
    Index 1 Param 1=&MakeCustomFloatFromBool(Bool=$User.IsFrontEnd, True Float=0, False Float=0)
    Index 1 Param 2=&MakeCustomFloatFromBool(Bool=$User.IsFrontEnd, True Float=2, False Float=1)
    Index 1 Param 3=&MakeCustomFloatFromBool(Bool=$User.IsFrontEnd, True Float=0, False Float=0)
    Index 2 Param 1=&MakeFloatFromVector(Vector=$Particles.FollowingDeform.Moving Average, Channel=::NewEnumerator0)
    Index 2 Param 2=&MakeFloatFromVector(Vector=$Particles.FollowingDeform.Moving Average, Channel=::NewEnumerator1)
    Index 2 Param 3=&MakeFloatFromVector(Vector=$Particles.FollowingDeform.Moving Average, Channel=::NewEnumerator2)
    Index 2 Param 4=$Particles.FollowingDeform.StretchDistance
    Index 3 Param 1=&MakeFloatFromVector(Vector=$Particles.FollowingDeform.ConstraintAxisWS, Channel=::NewEnumerator0)
    Index 3 Param 2=&MakeFloatFromVector(Vector=$Particles.FollowingDeform.ConstraintAxisWS, Channel=::NewEnumerator1)
    Index 3 Param 3=&MakeFloatFromVector(Vector=$Particles.FollowingDeform.ConstraintAxisWS, Channel=::NewEnumerator2)
    Module.Param0WriteEnabled=true
    Module.Param0WriteEnabled1=true
    Module.Param0WriteEnabled2=true
    Module.Param0WriteEnabled3=true
    Module.Param1WriteEnabled=false
    Module.Param1WriteEnabled1=true
    Module.Param1WriteEnabled2=true
    Module.Param1WriteEnabled3=true
    Module.Param2WriteEnabled=true
    Module.Param2WriteEnabled1=true
    Module.Param2WriteEnabled2=true
    Module.Param2WriteEnabled3=true
    Module.Param3WriteEnabled=false
    Module.Param3WriteEnabled1=false
    Module.Param3WriteEnabled2=true
    Module.Param3WriteEnabled3=true
    Write Parameter Index 1=true
    Write Parameter Index 2=true
    Write Parameter Index 3=true
   )
   &ScaleColor(
    Scale Alpha=&FloatFromCurve(
     FloatCurve=DataInterface() {propertyValues {Curve {keys [{arriveTangent=0.0842, interpMode=RCIM_Cubic, leaveTangent=0.0842, tangentMode=RCTM_User, time=1.7, value=0}, <at 2.5 1>]}}}
     CurveIndex=&Add_Float(A=$System.Age, B=&Multiply_FloatByInt(Float=100, Integer=$User.Deployed))
     Scale Curve=&Lerp_Float(A=0.5, B=1, Alpha=&ClampFloat(Float=$System.InterpolateOverTime.InterpolateOverTime_Float.Moving Average))
    )
    ScaleA=true
    ScaleRGB=true
   )
   /SetVariables(Particles.R2Color:LinearColor=&MakeLinearColor(
     R:NiagaraFloat=1
     G:NiagaraFloat=1
     B:NiagaraFloat=1
     A:NiagaraFloat=&FloatFromCurve(
      FloatCurve=DataInterface() {propertyValues {Curve {keys [
          {arriveTangent=1.2332, interpMode=RCIM_Cubic, leaveTangent=1.2332, tangentMode=RCTM_User, time=1.7, value=0}
          {arriveTangent=-0.0405, interpMode=RCIM_Cubic, leaveTangent=-0.0405, tangentMode=RCTM_User, time=2.5, value=1}
         ]}}}
      CurveIndex=&Add_Float(A=$System.Age, B=&Multiply_FloatByInt(Float=100, Integer=$User.Deployed))
      Scale Curve=&Lerp_Float(
       A=&MakeCustomFloatFromBool(Bool=$User.IsFrontEnd, True Float=0.7, False Float=0.5)
       B=1
       Alpha=&ClampFloat(Float=$System.InterpolateOverTime.InterpolateOverTime_Float.Moving Average)
      )
     )
    )) {isSetParameters=true}
   /SetVariables(Particles.PosForRead:NiagaraPosition=$Particles.FollowingDeform.FollowingPos) {isSetParameters=true}
   /SetVariables(Particles.FEScale:Vector3f=&MakeCustomVectorFromBool(
     Bool=$User.IsFrontEnd
     True Vector=&Multiply_VectorByFloat(Vector=$Particles.Scale, Float=0.85)
     False Vector=$Particles.Scale
    )) {isSetParameters=true}
   /SetVariables(Particles.New Position:NiagaraPosition=&Add_Vector(
     A=$Particles.FollowingDeform.RootPosition
     B=&MakeCustomVectorFromBool(Bool=$User.IsFrontEnd, True Vector=<0 0 20>, False Vector=<0 0 0>)
    )) {isSetParameters=true}
   &CameraOffset(Camera Offset Amount=&MakeCustomFloatFromBool(
     Bool=&IsPlatformSetActive(New Platform Set=DataInterface() {propertyValues {Platforms {
         qualityLevelMask=-31
         deviceProfileStates [
          {profileName=Mobile, qualityLevelMask=4294967295, setQualityLevelMask=4294967295}
          {profileName=Switch, qualityLevelMask=4294967295, setQualityLevelMask=4294967295}
         ]
        }}})
     True Float=-100
     False Float=0
    ))
  ]
  rendererClasses [&NiagaraMeshRendererProperties]
  renderers [
   {rendererClass=&NiagaraMeshRendererProperties, rendererIndex=0}
   {rendererClass=&NiagaraMeshRendererProperties, rendererIndex=1}
   {rendererClass=&NiagaraMeshRendererProperties, rendererIndex=2}
  ]
 }
}
summary {
 systemName=NS_EM_GliderMain
 emitters [
  {emitterName=Cloud003, rendererClasses [&NiagaraMeshRendererProperties]}
  {emitterName=FN_FollowingDeformTemplate001, rendererClasses [&NiagaraMeshRendererProperties]}
  {emitterName=Empty, rendererClasses [&NiagaraSpriteRendererProperties]}
 ]
 userVariables [
  User.Deployed {type=NiagaraInt32}
  User.ForwardDot {type=NiagaraFloat, defaultValue=0, description=""}
  User.IsFrontEnd {type=NiagaraBool}
 ]
}
