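# Niagara deploy effect for the Ensemble Maroon Male glider: a roughly two-second buildup of swirling sand-warp meshes, airflow and sprites.
# Curves on system age drive spawn rates and growth across five continuous emitters, including falling and stretched flipbook sprites.
# Two further mesh emitters burst at 1.7 seconds as the swirl swells.
# User.IsFrontEnd switches material parameters and airflow color, which otherwise follows time of day.
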NiagaraSystem /BRCosmetics/Gadgets/Assets/VinderTech_GliderChute/Glider_Ensemble_Maroon_Male/FX/NS_EM_Glider_DeployA
dependencies {
 usedDataInterfaces [&NiagaraDataInterfaceCurve, &NiagaraDataInterfaceVector2DCurve]
 usedDynamicInputs []
 usedModules [&SystemState]
}
emitters {
 Convergent_airflow_start {
  emitterUpdateScript [
   &EmitterState()
   &SpawnRate(
    SpawnRate=&FloatFromCurve(
     FloatCurve=DataInterface() {propertyValues {Curve {keys [
         {interpMode=RCIM_Cubic, tangentMode=RCTM_Break, time=0.0588, value=-0.0114}
         {interpMode=RCIM_Cubic, tangentMode=RCTM_Break, time=0.2588, value=0.9886}
         {interpMode=RCIM_Cubic, tangentMode=RCTM_Break, time=0.7035, value=1}
         {interpMode=RCIM_Cubic, tangentMode=RCTM_Break, time=1.0035, value=0}
        ]}}}
     CurveIndex=$System.Age
     Scale Curve=25
    )
    Use Spawn Probability=false
   )
  ]
  particleSpawnScript [
   /Niagara/Modules/Spawn/Initialization/InitializeParticle(
    Color=&MakeCustomLinearColorFromBool(Bool=$User.IsFrontEnd, True Color=<rgba 0.3 0.205 0.15 3>, False Color=<rgba 0.25 0.1762 0.125 3.5>)
    Lifetime=&RandomRangeFloat(Minimum=0.75, Maximum=0.95)
    Mesh Scale=&Multiply_VectorByFloat(Vector=&RandomRangeVector(Minimum=<0.7 0.7 0.7>, Maximum=<1 1 1>), Float=2)
    Module.Write Color=true
    Module.Write Lifetime=true
    Module.Write Mass=true
    Module.Write Material Random=true
    Module.Write Position=false
    Module.Write Scale=true
    Module.Write SpriteRotation=false
    Module.Write SpriteSize=true
    Position=&SimulationPosition()
    Sprite Size=<5 5>
   )
   &InitialMeshOrientation(Use Rotation Vector=false)
  ]
  particleUpdateScript [
   &ParticleState()
   &SolveForcesAndVelocity(Clamp Velocity=false, Limit Acceleration=false)
   &DynamicMaterialParameters(
    Index 0 Param 1=&FloatFromCurve(FloatCurve=DataInterface() {propertyValues {Curve {keys [<at 0 -0.2>,<at 1 1.2>]}}})
    Index 0 Param 2=1
    Index 0 Param 3=&Multiply_Float(
     A=&RandomRangeFloat(Minimum=0.8, Maximum=1)
     B=&FloatFromCurve(FloatCurve=DataInterface() {propertyValues {Curve {keys [<at 0 1>,<at 1 0.5>]}}}, Scale Curve=4)
    )
    Module.Param0WriteEnabled=true
    Module.Param1WriteEnabled=true
    Module.Param2WriteEnabled=true
    Module.Param3WriteEnabled=true
   )
   &DynamicMaterialParameters(
    Index 0 Param 1=&FloatFromCurve(FloatCurve=DataInterface() {propertyValues {Curve {keys [
         {interpMode=RCIM_Cubic, leaveTangent=3.7543, tangentMode=RCTM_Break, time=0, value=-0.2}
         {arriveTangent=0.511, interpMode=RCIM_Cubic, leaveTangent=0.511, tangentMode=RCTM_User, time=0.1241, value=0.2299}
         {arriveTangent=0.5035, interpMode=RCIM_Cubic, leaveTangent=0.5035, tangentMode=RCTM_User, time=0.3916, value=0.2873}
         {arriveTangent=2.4939, interpMode=RCIM_Cubic, leaveTangent=2.4939, tangentMode=RCTM_User, time=0.5851, value=0.6672}
         <at 1 1.2>
        ]}}})
    Index 0 Param 2=1
    Index 0 Param 3=&Multiply_Float(
     A=&RandomRangeFloat(Minimum=0.8, Maximum=1)
     B=&FloatFromCurve(FloatCurve=DataInterface() {propertyValues {Curve {keys [<at 0 1>,<at 1 0.5>]}}}, Scale Curve=4)
    )
    Module.Param0WriteEnabled=true
    Module.Param1WriteEnabled=true
    Module.Param2WriteEnabled=true
    Module.Param3WriteEnabled=true
   ) {enabled=false, module=DynamicMaterialParameters001}
   /SetVariables(Particles.Color=&HueShiftLinearColor(
     Hue=&MakeCustomFloatFromBool(
      Bool=$User.IsFrontEnd
      True Float=0
      False Float=&FNTimeOfDayScalar(Night Scale=-0.02, Morning Scale=0.01, Day Scale=0, Evening Scale=0.01)
     )
     Saturation=&MakeCustomFloatFromBool(
      Bool=$User.IsFrontEnd
      True Float=0.25
      False Float=&FNTimeOfDayScalar(Night Scale=0.95, Morning Scale=1.5, Evening Scale=1.4)
     )
     Value=&MakeCustomFloatFromBool(Bool=$User.IsFrontEnd, True Float=1.8, False Float=1)
     Linear Color=$Particles.Initial.Color
    )) {isSetParameters=true}
   &ScaleColor(
    Color Value To Scale=$Particles.Color
    Scale Alpha=&Multiply_Float(
     A=&Multiply_Float(
      A=&Multiply_Float(A=1, B=1)
      B=&FloatFromCurve(FloatCurve=DataInterface() {propertyValues {Curve {keys [
           <at 0 0 RCIM_Cubic>
           <at 0.1912 0.9932>
           {arriveTangent=-1.2279, interpMode=RCIM_Cubic, leaveTangent=-1.2279, time=0.7879, value=1.0115}
           <at 1 0>
          ]}}})
     )
     B=&FloatFromCurve(FloatCurve=DataInterface() {propertyValues {Curve {keys [<at 0 0>,<at 1.2 0.5258>]}}}, CurveIndex=$Emitter.Age)
    )
    ScaleA=true
    ScaleRGB=true
   )
   &ScaleMeshSize(Scale Factor=&Multiply_Vector(
     A=<1 1 1>
     B=&VectorFromFloat(Value=&FloatFromCurve(
       FloatCurve=DataInterface() {propertyValues {Curve {keys [
           {arriveTangent=0.145, interpMode=RCIM_Cubic, leaveTangent=0.145, tangentMode=RCTM_User, time=0, value=0}
           <at 0.4 0.5>
           {arriveTangent=0.447, interpMode=RCIM_Cubic, leaveTangent=0.447, tangentMode=RCTM_User, time=2, value=1.2}
          ]}}}
       CurveIndex=$System.Age
      ))
    ))
   &CameraOffset(Camera Offset Amount=20)
   /SetVariables(Particles.New Position:NiagaraPosition=&Add_Vector(A=Linked() {linkedVariable {name=Particles.Position, type=NiagaraPosition}}, B=<0 0 -33>)) {isSetParameters=true}
  ]
  rendererClasses [&NiagaraMeshRendererProperties]
  renderers [{rendererClass=&NiagaraMeshRendererProperties, rendererIndex=0}]
 }
 Empty001 {
  emitterUpdateScript [
   &EmitterState()
   &SpawnRate(
    SpawnRate=&FloatFromCurve(
     FloatCurve=DataInterface() {propertyValues {Curve {keys [
         {interpMode=RCIM_Cubic, leaveTangent=0.9278, tangentMode=RCTM_Break, time=0.4, value=0}
         {interpMode=RCIM_Cubic, tangentMode=RCTM_Break, time=0.95, value=1}
         {interpMode=RCIM_Cubic, tangentMode=RCTM_Break, time=1.4, value=1}
         <at 1.9 0>
        ]}}}
     CurveIndex=$System.Age
     Scale Curve=10
    )
    Use Spawn Probability=false
   )
  ]
  particleSpawnScript [
   /Niagara/Modules/Spawn/Initialization/V2/InitializeParticle(
    Color=<rgba 1 1 1 0.17>
    Lifetime Max=1
    Lifetime Min=0.5
    Lifetime Mode=::NewEnumerator1
    Position Offset=<0 0 -35>
    Sprite Rotation Mode=::NewEnumerator2
    Sprite Size Mode=::NewEnumerator1
    Uniform Sprite Size Max=110.0298
    Uniform Sprite Size Min=85.9507
    UsePositionOffset=true
    Write Material Random=true
   )
   &SubUVAnimation(End Frame=4, SubUV Animation Mode:ENiagara_SubUVLookupMode=::NewEnumerator2)
   &ShapeLocation(MaskSpawnGroup=false, Override Seed=false, Sphere Radius=10, UseOwnerScale=true)
   &AddVelocity(Module.Use Falloff Distance=false, Override Seed=false, Velocity Mode=::NewEnumerator1, Velocity Speed=0)
   &ApplyInitialForces()
  ]
  particleUpdateScript [
   &ParticleState()
   &DynamicMaterialParameters(
    Index 0 Param 1=&FloatFromCurve(FloatCurve=DataInterface() {propertyValues {Curve {keys [<at 0 -1>,<at 1 1>]}}})
    Index 0 Param 2=&MakeCustomFloatFromBool(Bool=$User.IsFrontEnd, True Float=0, False Float=0)
    Index 0 Param 3=&MakeCustomFloatFromBool(Bool=$User.IsFrontEnd, True Float=2, False Float=1)
    Index 0 Param 4=&MakeCustomFloatFromBool(Bool=$User.IsFrontEnd, True Float=0, False Float=0)
    Module.Param0WriteEnabled=true
    Module.Param1WriteEnabled=true
    Module.Param2WriteEnabled=true
    Module.Param3WriteEnabled=true
   )
   &GravityForce(Gravity=&MakeVector(
     X=0
     Y=0
     Z=&FloatFromCurve(
      FloatCurve=DataInterface() {propertyValues {Curve {keys [
          {interpMode=RCIM_Cubic, leaveTangent=0.5369, tangentMode=RCTM_Break, time=0.35, value=0.5}
          {arriveTangent=0.341, interpMode=RCIM_Cubic, tangentMode=RCTM_Break, time=1.5, value=1}
         ]}}}
      CurveIndex=$System.Age
      Scale Curve=&RandomRangeFloat(Minimum=-650, Maximum=-750)
     )
    ))
   &Drag(Drag=0, Use Linear Drag=true, Use Rotational Drag=false)
   &SolveForcesAndVelocity(Clamp Velocity=false, Limit Acceleration=false)
   &ScaleSpriteSize(Uniform Curve Sprite Scale=DataInterface() {propertyValues {Curve {keys [
        {arriveTangent=7.257, interpMode=RCIM_Cubic, leaveTangent=7.257, tangentMode=RCTM_User, time=0, value=0}
        {arriveTangent=0.9995, interpMode=RCIM_Cubic, leaveTangent=0.9995, tangentMode=RCTM_User, time=0.1568, value=0.8364}
        {arriveTangent=-0.6962, interpMode=RCIM_Cubic, leaveTangent=-0.6962, tangentMode=RCTM_User, time=1, value=0.5}
       ]}}})
   &ScaleColor(
    Scale Alpha=&FloatFromCurve(FloatCurve=DataInterface() {propertyValues {Curve {keys [
         {arriveTangent=4.1785, interpMode=RCIM_Cubic, leaveTangent=4.1785, tangentMode=RCTM_User, time=0, value=0}
         {arriveTangent=0.9995, interpMode=RCIM_Cubic, leaveTangent=0.9995, tangentMode=RCTM_User, time=0.2223, value=0.75}
         {interpMode=RCIM_Cubic, tangentMode=RCTM_Break, time=0.8, value=1}
         {arriveTangent=-8.8798, interpMode=RCIM_Cubic, leaveTangent=-0.0405, tangentMode=RCTM_Break, time=1, value=0}
        ]}}})
    ScaleA=true
    ScaleRGB=true
   )
   &ScaleSpriteSize(
    Uniform Curve Index=$System.Age
    Uniform Curve Sprite Scale=DataInterface() {propertyValues {Curve {keys [{tangentMode=RCTM_User, time=0.3, value=0}, {tangentMode=RCTM_User, time=1.2, value=1}]}}}
   ) {module=ScaleSpriteSize001}
  ]
  rendererClasses [&NiagaraSpriteRendererProperties]
  renderers [{rendererClass=&NiagaraSpriteRendererProperties}]
 }
 Empty002 {
  emitterUpdateScript [
   &EmitterState()
   &SpawnRate(
    SpawnRate=&FloatFromCurve(
     FloatCurve=DataInterface() {propertyValues {Curve {keys [
         <at 0 0>
         <at 0.35 0>
         {interpMode=RCIM_Cubic, tangentMode=RCTM_Break, time=0.45, value=1}
         {arriveTangent=-3.6147, interpMode=RCIM_Cubic, tangentMode=RCTM_Break, time=1.5, value=0}
        ]}}}
     CurveIndex=$System.Age
     Scale Curve=20
    )
    Use Spawn Probability=false
   )
  ]
  particleSpawnScript [
   &InitializeParticle(
    Color Mode=::NewEnumerator0
    Lifetime Max=1.2
    Lifetime Min=0.8
    Lifetime Mode=::NewEnumerator1
    Mesh Scale Mode=::NewEnumerator1
    Mesh Uniform Scale Max=1
    Sprite Size Mode=::NewEnumerator1
    Uniform Sprite Size Max=25
    Uniform Sprite Size Min=15
    UsePositionOffset=false
    Write Material Random=true
   )
   &SphereLocation(
    Hemisphere Z=true
    MaskSpawnGroup=false
    Offset=<0 0 10>
    Override Seed=false
    Sphere Radius=7
    Surface Only=false
    UseNonUniformScale=false
    UseOffset=true
    UseSphereOrientation=false
   )
   /SetVariables(Particles.SpriteAlignment=<0 0 1>) {isSetParameters=true}
   &AddVelocity(Velocity=<0 0 -15>)
   &ApplyInitialForces()
   &SubUVAnimation(End Frame=4, SubUV Animation Mode:ENiagara_SubUVLookupMode=::NewEnumerator2)
  ]
  particleUpdateScript [
   &ParticleState()
   &ScaleSpriteSize(
    Non-Uniform Curve Sprite Scale=DataInterface() {propertyValues {
      XCurve {keys [{tangentMode=RCTM_User, time=0, value=0}, <at 0.1 0.1>, {tangentMode=RCTM_User, time=0.2, value=1}]}
      YCurve {keys [
        {arriveTangent=4.1785, interpMode=RCIM_Cubic, leaveTangent=4.1785, tangentMode=RCTM_User, time=0, value=0}
        {arriveTangent=0.9995, interpMode=RCIM_Cubic, leaveTangent=0.9995, tangentMode=RCTM_User, time=0.2223, value=2}
        {arriveTangent=-0.0405, interpMode=RCIM_Cubic, leaveTangent=-0.0405, tangentMode=RCTM_User, time=1, value=10}
       ]}
     }}
    Scale Sprite Size Mode=::NewEnumerator3
   )
   &DynamicMaterialParameters(
    Index 0 Param 1=&Add_Float(
     A=&self:ScratchDynamicInput(Float Max:NiagaraFloat=0, Float Min:NiagaraFloat=100, Seed:NiagaraInt32=15)
     B=$System.Age
    )
    Index 0 Param 2=&FloatFromCurve(FloatCurve=DataInterface() {propertyValues {Curve {keys [<at 0 0>,<at 1 -1>]}}})
    Index 0 Param 3=&MakeCustomFloatFromBool(Bool=$User.IsFrontEnd, True Float=0, False Float=0)
    Index 0 Param 4=&MakeCustomFloatFromBool(Bool=$User.IsFrontEnd, True Float=1.6, False Float=1)
    Module.Param0WriteEnabled=true
    Module.Param0WriteEnabled1=true
    Module.Param1WriteEnabled=true
    Module.Param1WriteEnabled1=true
    Module.Param2WriteEnabled=true
    Module.Param2WriteEnabled1=true
    Module.Param3WriteEnabled=true
    Module.Param3WriteEnabled1=true
    Write Parameter Index 1=true
   )
   &ScaleSpriteSize(Scale Factor=<5 0.7>, Scale Sprite Size Mode=::NewEnumerator2) {module=ScaleSpriteSize001}
   &GravityForce(Gravity=<0 0 -221.8937>) {enabled=false}
   &SolveForcesAndVelocity(Clamp Velocity=false, Limit Acceleration=false)
   /SetVariables(Particles.New Position:NiagaraPosition=&Add_Vector(A=Linked() {linkedVariable {name=Particles.Position, type=NiagaraPosition}}, B=<0 0 -43>)) {isSetParameters=true}
  ]
  rendererClasses [&NiagaraSpriteRendererProperties]
  renderers [{rendererClass=&NiagaraSpriteRendererProperties}]
 }
 SandWarp003 {
  emitterUpdateScript [
   &EmitterState()
   &SpawnRate(
    SpawnRate=&FloatFromCurve(FloatCurve=DataInterface() {propertyValues {Curve {keys [<at 0 1>,<at 1 1>,<at 1.5 1>,<at 1.7 0>]}}}, CurveIndex=$System.Age, Scale Curve=30)
    Use Spawn Probability=false
   )
  ]
  particleSpawnScript [
   &InitializeParticle(
    Color Mode=::NewEnumerator0
    Lifetime=0.5
    Mesh Scale Mode=::NewEnumerator1
    Mesh Uniform Scale Max=1.5
    Mesh Uniform Scale Min=1.5
    UsePositionOffset=false
    Write Material Random=true
   )
   &InitialMeshOrientation(
    Mesh Orientation Mode=::NewEnumerator7
    Rotation=&RandomRangeVector(Minimum=<-1 -1 -1>, Maximum=<1 1 1>)
    Use Rotation Vector=true
   )
  ]
  particleUpdateScript [
   &ParticleState()
   &DynamicMaterialParameters(
    Index 0 Param 1=&FloatFromCurve(FloatCurve=DataInterface() {propertyValues {Curve {keys [{interpMode=RCIM_Cubic, tangentMode=RCTM_User, time=0.5, value=0}, {interpMode=RCIM_Cubic, tangentMode=RCTM_User, time=1, value=1}]}}})
    Index 0 Param 2=&MakeCustomFloatFromBool(Bool=$User.IsFrontEnd, True Float=0, False Float=0)
    Index 0 Param 3=&MakeCustomFloatFromBool(Bool=$User.IsFrontEnd, True Float=1.6, False Float=1)
    Index 0 Param 4=&MakeCustomFloatFromBool(Bool=$User.IsFrontEnd, True Float=0, False Float=0)
    Module.Param0WriteEnabled=true
    Module.Param1WriteEnabled=true
    Module.Param2WriteEnabled=true
    Module.Param3WriteEnabled=true
   )
   &ScaleMeshSize(Scale Factor=&VectorFromFloat(Value=&FloatFromCurve(FloatCurve=DataInterface() {propertyValues {Curve {keys [
          {arriveTangent=2.4702, interpMode=RCIM_Cubic, leaveTangent=2.4702, tangentMode=RCTM_User, time=0, value=1}
          {arriveTangent=-0.0758, interpMode=RCIM_Cubic, leaveTangent=-0.0758, tangentMode=RCTM_User, time=0.4, value=1}
          {arriveTangent=0.4623, interpMode=RCIM_Cubic, leaveTangent=0.4623, tangentMode=RCTM_User, time=1, value=0.5}
         ]}}})))
   &UpdateMeshOrientation(
    Rotation Rate=&FloatFromCurve(
     FloatCurve=DataInterface() {propertyValues {Curve {keys [
         {interpMode=RCIM_Cubic, leaveTangent=0.6814, tangentMode=RCTM_Break, time=0.5, value=0}
         {arriveTangent=0.8228, interpMode=RCIM_Cubic, tangentMode=RCTM_Break, time=1.5, value=1}
        ]}}}
     CurveIndex=$System.Age
     Scale Curve=4.5
    )
    Rotation Vector=<0 0 1>
   ) {enabled=false}
   &UpdateMeshOrientation(Rotation Rate=4.5, Rotation Vector=<0 0 1>) {module=UpdateMeshOrientation001}
   &ScaleMeshSize(Scale Factor=&VectorFromFloat(Value=&FloatFromCurve(
      FloatCurve=DataInterface() {propertyValues {Curve {keys [
          {interpMode=RCIM_Cubic, tangentMode=RCTM_Break, time=0, value=0}
          {arriveTangent=0.3579, interpMode=RCIM_Cubic, leaveTangent=0.3579, tangentMode=RCTM_User, time=0.9, value=0.25}
          {arriveTangent=0.2353, interpMode=RCIM_Cubic, leaveTangent=0.2353, tangentMode=RCTM_User, time=1.7, value=0.43}
          {arriveTangent=2.3824, interpMode=RCIM_Cubic, leaveTangent=2.3824, tangentMode=RCTM_User, time=1.9, value=1.2}
          <at 3 2>
         ]}}}
      CurveIndex=$System.Age
     ))) {module=ScaleMeshSize001}
   /SetVariables(Particles.New Position:NiagaraPosition=&Add_Vector(A=Linked() {linkedVariable {name=Particles.Position, type=NiagaraPosition}}, B=<0 0 -25>)) {isSetParameters=true}
  ]
  rendererClasses [&NiagaraMeshRendererProperties]
  renderers [{rendererClass=&NiagaraMeshRendererProperties, rendererIndex=0}]
 }
 SandWarp004 {
  emitterUpdateScript [
   &EmitterState()
   &SpawnRate(
    SpawnRate=&FloatFromCurve(FloatCurve=DataInterface() {propertyValues {Curve {keys [<at 0 1>,<at 1 1>,<at 1.8 1>,<at 2 0>]}}}, CurveIndex=$System.Age, Scale Curve=10)
    Use Spawn Probability=false
   ) {module=SpawnRate001}
  ]
  particleSpawnScript [
   &InitializeParticle(
    Color=<rgba 1 1 1 1.2>
    Lifetime=0.7
    Mesh Scale Mode=::NewEnumerator1
    Mesh Uniform Scale Max=3
    Mesh Uniform Scale Min=3
    UsePositionOffset=false
    Write Material Random=true
   )
   &InitialMeshOrientation(
    Mesh Orientation Mode=::NewEnumerator7
    Rotation=&Lerp_Vector(
     A=&RandomRangeVector(Minimum=<-1 -1 0>, Maximum=<1 1 1>)
     B=&RandomRangeVector(Minimum=<-0.125 -0.125 0>, Maximum=<0.125 0.125 1>)
     Alpha=&FloatFromCurve(FloatCurve=DataInterface() {propertyValues {Curve {keys [<at 0 0>,<at 1.5 1>]}}}, CurveIndex=$System.Age)
    )
    Use Rotation Vector=true
   )
  ]
  particleUpdateScript [
   &ParticleState()
   &DynamicMaterialParameters(
    Index 0 Param 1=&FloatFromCurve(FloatCurve=DataInterface() {propertyValues {Curve {keys [{interpMode=RCIM_Cubic, tangentMode=RCTM_User, time=0.6, value=0}, {interpMode=RCIM_Cubic, tangentMode=RCTM_User, time=1, value=1}]}}})
    Index 0 Param 2=&MakeCustomFloatFromBool(Bool=$User.IsFrontEnd, True Float=0, False Float=0)
    Index 0 Param 3=&MakeCustomFloatFromBool(Bool=$User.IsFrontEnd, True Float=2, False Float=1)
    Index 0 Param 4=&MakeCustomFloatFromBool(Bool=$User.IsFrontEnd, True Float=0, False Float=0)
    Module.Param0WriteEnabled=true
    Module.Param1WriteEnabled=true
    Module.Param2WriteEnabled=true
    Module.Param3WriteEnabled=true
   )
   /SetVariables(
    Particles.XY:NiagaraFloat=&FloatFromCurve(FloatCurve=DataInterface() {propertyValues {Curve {keys [
         {arriveTangent=0.6321, interpMode=RCIM_Cubic, leaveTangent=0.6321, tangentMode=RCTM_User, time=0, value=0.2}
         {arriveTangent=0.4623, interpMode=RCIM_Cubic, leaveTangent=0.4623, tangentMode=RCTM_User, time=1, value=1}
        ]}}})
    Particles.Z:NiagaraFloat=&FloatFromCurve(
     FloatCurve=DataInterface() {propertyValues {Curve {keys [
         {arriveTangent=1.8479, interpMode=RCIM_Cubic, leaveTangent=1.8479, tangentMode=RCTM_User, time=0, value=0}
         {arriveTangent=0.9995, interpMode=RCIM_Cubic, leaveTangent=0.9995, tangentMode=RCTM_User, time=0.4, value=0.4}
         {arriveTangent=0.4855, interpMode=RCIM_Cubic, leaveTangent=0.4855, tangentMode=RCTM_User, time=1, value=0.6}
        ]}}}
     Scale Curve=&FloatFromCurve(FloatCurve=DataInterface() {propertyValues {Curve {keys [<at 0 3>,<at 1.5 1>]}}}, CurveIndex=$System.Age)
    )
   ) {isSetParameters=true, module=SetVariables_#1}
   &ScaleMeshSize(Scale Factor=&MakeVector(X=$Particles.XY, Y=$Particles.XY, Z=$Particles.Z))
   &UpdateMeshOrientation(Rotation Rate=1.5, Rotation Vector=<0 0 1>)
   &ScaleMeshSize(Scale Factor=&MakeVector(
     X=1
     Y=1
     Z=&FloatFromCurve(
      FloatCurve=DataInterface() {propertyValues {Curve {keys [{interpMode=RCIM_Cubic, tangentMode=RCTM_Break, time=0, value=1}, <at 1 2.5>]}}}
      Scale Curve=1.5
     )
    )) {module=ScaleMeshSize001}
   &ScaleColor(
    Scale Alpha=&FloatFromCurve(FloatCurve=DataInterface() {propertyValues {Curve {keys [
         {arriveTangent=8.2356, interpMode=RCIM_Cubic, leaveTangent=8.2356, tangentMode=RCTM_User, time=0, value=0}
         {arriveTangent=0.1969, interpMode=RCIM_Cubic, leaveTangent=0.1969, tangentMode=RCTM_User, time=0.2223, value=1}
         {arriveTangent=-2.4818, interpMode=RCIM_Cubic, leaveTangent=-2.4818, tangentMode=RCTM_User, time=1, value=0}
        ]}}})
    ScaleA=true
    ScaleRGB=true
   )
   &ScaleMeshSize(Scale Factor=&VectorFromFloat(Value=&FloatFromCurve(
      FloatCurve=DataInterface() {propertyValues {Curve {keys [
          <at 0 0 RCIM_Cubic>
          {arriveTangent=0.2404, interpMode=RCIM_Cubic, leaveTangent=0.2404, tangentMode=RCTM_User, time=0.9, value=0.25}
          {arriveTangent=0.3821, interpMode=RCIM_Cubic, leaveTangent=0.3821, tangentMode=RCTM_User, time=1.7, value=0.43}
          <at 1.9 1.7>
          <at 2.2 2>
         ]}}}
      CurveIndex=$System.Age
     ))) {module=ScaleMeshSize002}
   &ScaleColor(
    Scale Alpha=&FloatFromCurve(FloatCurve=DataInterface() {propertyValues {Curve {keys [<at 0 1>,<at 1.65 1>,<at 1.8 0.6>]}}}, CurveIndex=$System.Age)
    ScaleA=true
    ScaleRGB=true
   ) {module=ScaleColor001}
   /SetVariables(Particles.New Position:NiagaraPosition=&Add_Vector(A=Linked() {linkedVariable {name=Particles.Position, type=NiagaraPosition}}, B=<0 0 -25>)) {isSetParameters=true}
  ]
  rendererClasses [&NiagaraMeshRendererProperties]
  renderers [{rendererClass=&NiagaraMeshRendererProperties, rendererIndex=0}]
 }
 SandWarp005 {
  emitterUpdateScript [
   &EmitterState()
   &SpawnBurst_Instantaneous(Spawn Count=4, Spawn Time=1.7, Use Spawn Probability=false, UseLoopCountLimit=false)
  ]
  particleSpawnScript [
   &InitializeParticle(
    Color=<rgba 1 0.881 0.8 0.4>
    Lifetime=1.4
    Mesh Scale Mode=::NewEnumerator1
    Mesh Uniform Scale Max=16
    Mesh Uniform Scale Min=15
    UsePositionOffset=false
    Write Material Random=true
   )
   &InitialMeshOrientation(
    Mesh Orientation Mode=::NewEnumerator7
    Rotation=&RandomRangeVector(Minimum=<-0.0625 -0.0625 0>, Maximum=<0.0625 0.0625 1>)
    Use Rotation Vector=true
   )
  ]
  particleUpdateScript [
   &ParticleState()
   &DynamicMaterialParameters(
    Index 0 Param 1=&FloatFromCurve(FloatCurve=DataInterface() {propertyValues {Curve {keys [{interpMode=RCIM_Cubic, tangentMode=RCTM_User, time=0.5, value=0}, {interpMode=RCIM_Cubic, tangentMode=RCTM_User, time=1, value=1}]}}})
    Index 0 Param 2=&MakeCustomFloatFromBool(Bool=$User.IsFrontEnd, True Float=0, False Float=0)
    Index 0 Param 3=&MakeCustomFloatFromBool(Bool=$User.IsFrontEnd, True Float=2, False Float=1)
    Index 0 Param 4=&MakeCustomFloatFromBool(Bool=$User.IsFrontEnd, True Float=0, False Float=0)
    Module.Param0WriteEnabled=true
    Module.Param1WriteEnabled=true
    Module.Param2WriteEnabled=true
    Module.Param3WriteEnabled=true
   )
   /SetVariables(
    Particles.XY:NiagaraFloat=&FloatFromCurve(
     FloatCurve=DataInterface() {propertyValues {Curve {keys [
         {arriveTangent=0.6321, interpMode=RCIM_Cubic, leaveTangent=0.6321, tangentMode=RCTM_User, time=0, value=0}
         {arriveTangent=0.4623, interpMode=RCIM_Cubic, leaveTangent=0.4623, tangentMode=RCTM_User, time=1, value=1}
        ]}}}
     CurveIndex=$Particles.Age
    )
    Particles.Z:NiagaraFloat=&FloatFromCurve(
     FloatCurve=DataInterface() {propertyValues {Curve {keys [
         {arriveTangent=1.8479, interpMode=RCIM_Cubic, leaveTangent=1.8479, tangentMode=RCTM_User, time=0, value=0}
         {arriveTangent=0.9995, interpMode=RCIM_Cubic, leaveTangent=0.9995, tangentMode=RCTM_User, time=0.4, value=0.4}
         {arriveTangent=0.4855, interpMode=RCIM_Cubic, leaveTangent=0.4855, tangentMode=RCTM_User, time=1, value=0.6}
        ]}}}
     CurveIndex=$Particles.Age
    )
   ) {isSetParameters=true, module=SetVariables_#1}
   &ScaleMeshSize(Scale Factor=&MakeVector(X=$Particles.XY, Y=$Particles.XY, Z=$Particles.Z))
   &UpdateMeshOrientation(Rotation Rate=1.2, Rotation Vector=<0 0 1>)
   &ScaleMeshSize(Scale Factor=&MakeVector(
     X=1
     Y=1
     Z=&FloatFromCurve(
      FloatCurve=DataInterface() {propertyValues {Curve {keys [{interpMode=RCIM_Cubic, tangentMode=RCTM_Break, time=0, value=1}, <at 1 1.5>]}}}
      CurveIndex=$Particles.Age
      Scale Curve=1.2
     )
    )) {module=ScaleMeshSize001}
   &ScaleColor(
    Scale Alpha=&FloatFromCurve(
     FloatCurve=DataInterface() {propertyValues {Curve {keys [
         {arriveTangent=8.2356, interpMode=RCIM_Cubic, leaveTangent=8.2356, tangentMode=RCTM_User, time=0, value=0}
         {arriveTangent=0.1969, interpMode=RCIM_Cubic, leaveTangent=0.1969, tangentMode=RCTM_User, time=0.2223, value=1}
         {arriveTangent=-1.2858, interpMode=RCIM_Cubic, leaveTangent=-1.2858, time=0.5, value=0.6}
         {arriveTangent=-0.7636, interpMode=RCIM_Cubic, leaveTangent=-0.7636, tangentMode=RCTM_User, time=1, value=0}
        ]}}}
     Scale Curve=0.5
    )
    ScaleA=true
    ScaleRGB=true
   )
   /SetVariables(Particles.New Position:NiagaraPosition=&Add_Vector(A=Linked() {linkedVariable {name=Particles.Position, type=NiagaraPosition}}, B=<0 0 -25>)) {isSetParameters=true}
  ]
  rendererClasses [&NiagaraMeshRendererProperties]
  renderers [{rendererClass=&NiagaraMeshRendererProperties, rendererIndex=0}]
 }
 SandWarp006 {
  emitterUpdateScript [
   &EmitterState()
   &SpawnBurst_Instantaneous(Spawn Count=3, Spawn Time=1.7, Use Spawn Probability=false, UseLoopCountLimit=false)
  ]
  particleSpawnScript [
   &InitializeParticle(
    Lifetime=0.5
    Mesh Scale Mode=::NewEnumerator1
    Mesh Uniform Scale Max=6
    Mesh Uniform Scale Min=3
    UsePositionOffset=false
    Write Material Random=true
   )
   &InitialMeshOrientation(
    Mesh Orientation Mode=::NewEnumerator7
    Rotation=&RandomRangeVector(Minimum=<-0.125 -0.125 0>, Maximum=<0.125 0.125 1>)
    Use Rotation Vector=true
   ) {enabled=false}
   &InitialMeshOrientation(
    Mesh Orientation Mode=::NewEnumerator7
    Rotation=&RandomRangeVector(Minimum=<-1 -1 0>, Maximum=<1 1 1>)
    Use Rotation Vector=true
   ) {module=InitialMeshOrientation001}
  ]
  particleUpdateScript [
   &ParticleState()
   &DynamicMaterialParameters(
    Index 0 Param 1=&FloatFromCurve(FloatCurve=DataInterface() {propertyValues {Curve {keys [{interpMode=RCIM_Cubic, tangentMode=RCTM_User, time=0.5, value=0}, {interpMode=RCIM_Cubic, tangentMode=RCTM_User, time=1, value=1}]}}})
    Index 0 Param 2=&MakeCustomFloatFromBool(Bool=$User.IsFrontEnd, True Float=0, False Float=0)
    Index 0 Param 3=&MakeCustomFloatFromBool(Bool=$User.IsFrontEnd, True Float=1.6, False Float=1)
    Index 0 Param 4=&MakeCustomFloatFromBool(Bool=$User.IsFrontEnd, True Float=0, False Float=0)
    Module.Param0WriteEnabled=true
    Module.Param1WriteEnabled=true
    Module.Param2WriteEnabled=true
    Module.Param3WriteEnabled=true
   )
   &ScaleMeshSize(Scale Factor=&VectorFromFloat(Value=&FloatFromCurve(FloatCurve=DataInterface() {propertyValues {Curve {keys [
          {arriveTangent=1.8479, interpMode=RCIM_Cubic, leaveTangent=1.8479, tangentMode=RCTM_User, time=0, value=0}
          {arriveTangent=0.9995, interpMode=RCIM_Cubic, leaveTangent=0.9995, tangentMode=RCTM_User, time=0.4, value=0.5}
          {arriveTangent=0.4623, interpMode=RCIM_Cubic, leaveTangent=0.4623, tangentMode=RCTM_User, time=1, value=1}
         ]}}})))
   &UpdateMeshOrientation(Rotation Rate=2, Rotation Vector=<0 0 1>)
   /SetVariables(Particles.New Position:NiagaraPosition=&Add_Vector(A=Linked() {linkedVariable {name=Particles.Position, type=NiagaraPosition}}, B=<0 0 -25>)) {isSetParameters=true}
  ]
  rendererClasses [&NiagaraMeshRendererProperties]
  renderers [{rendererClass=&NiagaraMeshRendererProperties, rendererIndex=0}]
 }
}
summary {
 systemName=NS_EM_Glider_DeployA
 emitters [
  {emitterName=SandWarp003, rendererClasses [&NiagaraMeshRendererProperties]}
  {emitterName=SandWarp004, rendererClasses [&NiagaraMeshRendererProperties]}
  {emitterName=SandWarp005, rendererClasses [&NiagaraMeshRendererProperties]}
  {emitterName=SandWarp006, rendererClasses [&NiagaraMeshRendererProperties]}
  {emitterName=Empty001, rendererClasses [&NiagaraSpriteRendererProperties]}
  {emitterName=Empty002, rendererClasses [&NiagaraSpriteRendererProperties]}
  {emitterName=Convergent_airflow_start, rendererClasses [&NiagaraMeshRendererProperties]}
 ]
 userVariables [User.IsFrontEnd {type=NiagaraBool}]
}
